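# Control Rig for the F_MED_LacedCrimp female player body cosmetic, adding corrective deformation and a belt adjustment to its skeleton.
# A sequence first runs the shared DeformRig function for limb, groin and glute helper bones.
# A spherical pose reader then tracks spine_03 against the pelvis, weighting a transform on dyn_belt_mstr.
# That step hangs off the isCrouching branch's Completed pin, so it runs regardless of crouching; its debug drawing is enabled.
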ControlRigBlueprint /BRCosmetics/Characters/Player/Female/Medium/Bodies/F_MED_LacedCrimp/Meshes/F_MED_LacedCrimp_CtrlRig
graphs [
 RigVMFunctionLibrary {links [], nodes []}
 RigVMFunction_Sequence_1_ContainedGraph {nodes [
   n1 = RigVMFunctionEntryNode()
   RigVMFunction_Sequence = &RigVMFunction_Sequence(ExecuteContext=@n1.ExecuteContext)
   n2 = &RigVMFunction_Sequence(ExecuteContext=@RigVMFunction_Sequence.B)
   n3 = &RigVMFunction_Sequence(ExecuteContext=@n2.B)
   RigVMFunctionReturnNode(A=@RigVMFunction_Sequence.A, B=@n2.A, C=@n3.A, D=@n3.B)
  ]}
 RigVMModel {nodes [
   RigUnit_BeginExecution = &RigUnit_BeginExecution()
   RigVMFunction_Sequence = RigVMAggregateNode(ExecuteContext=@RigUnit_BeginExecution.ExecutePin)
   FN_DeformRig_FNC = RigVMFunctionReferenceNode(
    DeformRig_ Curve="(deformRig_l_upperarm_up_alpha_72_#1=False,deformRig_r_upperarm_up_alpha_73_#2=False,deformRig_l_upperarm_fr_alpha_74_#3=False,deformRig_r_upperarm_fr_alpha_75_CB51B8 ...(truncated)"
    DeformRig_Bool="(upperarm_1_#4=true,elbow_3_#5=true,wrist_7_#6=false,groin_8_#7=true,glute_17_#8=true,knee_9_40C408324CA26 ...(truncated)"
    DeformRig_Debug=(upperarm_debug_18_#4=False,wrist_debug_20_#6=False,lat_debug_26_#9=False,groinGlute_debug_27_#7=False)
    DeformRig_Value="(upperarm_up_18_#10=(Rotation=(X=0.0,Y=-0.0,Z=0.0,W=1.0),Translation=(X=0.0,Y=0.0,Z=12),Scale3D=(X=0.70,Y=1.0,Z=1.0)),upperarm_fwd_fr_19_#11=(Rotation=(X=0.0,Y=-0.0,Z=0.0,W=1.0),Translation=(X=0.0,Y ...(truncated)"
    ExecuteContext=@RigVMFunction_Sequence.A
   )
   n1 = &RigVMFunction_ControlFlowBranch(ExecuteContext=@RigVMFunction_Sequence.B, Condition=RigVMVariableNode() {variable=isCrouching}.Value)
   n2 = &RigUnit_SphericalPoseReader(
    ActiveRegionScaleFactors=(PositiveWidth=1,NegativeWidth=1,PositiveHeight=1,NegativeHeight=1)
    ActiveRegionSize=0.15
    Debug=(bDrawDebug=True,bDraw2D=True,bDrawLocalAxes=True,DebugScale=25,DebugSegments=20,DebugThickness=0.25)
    DriverAxis=(X=2.0198,Y=-2.0194,Z=11.9603)
    DriverItem=(Type=Bone,Name="spine_03")
    FalloffRegionScaleFactors=(PositiveWidth=1,NegativeWidth=1,PositiveHeight=1,NegativeHeight=1)
    FalloffSize=0.4
    FlipHeightScaling=False
    FlipWidthScaling=False
    OptionalParentItem=(Type=Bone,Name="pelvis")
    RotationOffset=(X=0,Y=0,Z=0)
    ExecutePin=@n1.Completed
   )
   > &RigUnit_ModifyTransforms(
    ItemToModify=((Item=(Type=Bone,Name="dyn_belt_mstr"),Transform=(Rotation=(X=0.0109,Y=-0.104,Z=0.104,W=0.9891),Translation=(X=0,Y=0,Z=0),Scale3D=(X=1,Y=1,Z=1))))
    WeightMinimum=0
    Weight=@n2.OutputParam
   )
  ]}
]
hierarchy [
 root {loc=<0 0 0>, rot=<0 0 0 1>}
 pelvis {parent=root, loc=<0 -0.5301 98.0571>, rot=<0.0127 0.707 -0.0127 -0.707>}
 spine_01 {parent=pelvis, loc=<6.1208 -0.9328 0>, rot=<0 0 0.034 -0.9994>}
 spine_02 {parent=spine_01, loc=<6.5997 -0.2095 0>, rot=<0 0 0.0238 -0.9997>}
 spine_03 {parent=spine_02, loc=<6.4914 -0.1237 0>, rot=<0 0 0.0149 -0.9999>}
 spine_04 {parent=spine_03, loc=<6.9598 0.0794 0>, rot=<0 0 -0.0132 -0.9999>}
 spine_05 {parent=spine_04, loc=<6.9987 0 0>, rot=<0 0 -0.0015 -1>}
 neck_01 {parent=spine_05, loc=<11.3138 0.1529 0>, rot=<0 0 -0.2003 -0.9797>}
 neck_02 {parent=neck_01, loc=<4.5627 0 0>, rot=<0 0 0.0504 -0.9987>}
 head {parent=neck_02, loc=<5.3914 0 0>, rot=<0 0 0.115 -0.9934>}
 mstr_ring_A {parent=neck_01, loc=<0.9592 5.3481 1.3309>, rot=<-0.6905 -0.7098 -0.1239 0.0641>}
 dyn_ring_A {parent=mstr_ring_A, loc=<0 0 0>, rot=<0 0 0 1>}
 mstr_ring_B {parent=neck_01, loc=<0.865 5.6399 -0.0436>, rot=<-0.6635 -0.7387 0.1183 -0.011>}
 dyn_ring_B {parent=mstr_ring_B, loc=<0 0 0>, rot=<0 0 0 1>}
 mstr_ring_C {parent=neck_01, loc=<0.9515 5.4758 -1.3748>, rot=<-0.6716 -0.716 0.1326 -0.1371>}
 dyn_ring_C {parent=mstr_ring_C, loc=<0 0 0>, rot=<0 0 0 1>}
 clavicle_l {parent=spine_05, loc=<5.4831 -0.1053 -3.2946>, rot=<0.0465 -0.7289 -0.009 -0.6829>}
 upperarm_l {parent=clavicle_l, loc=<9.8658 -1.2136 -0.4105>, rot=<0.0164 -0.3245 -0.0269 -0.9454>}
 lowerarm_l {parent=upperarm_l, loc=<26.465 -0.8345 0.259>, rot=<0.0036 0.0161 -0.2298 -0.9731>}
 hand_l {parent=lowerarm_l, loc=<22.7355 -0.3203 -0.6105>, rot=<-0.6784 -0.0223 0.0735 -0.7307>}
 thumb_01_l {parent=hand_l, loc=<1.3118 -0.3191 -2.459>, rot=<0.6546 -0.4042 0.0664 -0.6354>}
 thumb_02_l {parent=thumb_01_l, loc=<5.9336 -0.0448 0.0401>, rot=<-0.0037 0.0213 0.298 -0.9543>}
 thumb_03_l {parent=thumb_02_l, loc=<2.2223 0.3682 -0.0907>, rot=<0.0118 -0.013 0.0066 -0.9998>}
 index_metacarpal_l {parent=hand_l, loc=<1.6535 0.0183 -2.96>, rot=<0.0158 -0.0124 -0.0173 -0.9996>}
 index_01_l {parent=index_metacarpal_l, loc=<8.3008 0.0357 -0.0524>, rot=<-0.0044 0.0485 0.0924 -0.9945>}
 index_02_l {parent=index_01_l, loc=<3.3587 -0.3004 0.0452>, rot=<-0.002 0.0027 0.3225 -0.9466>}
 index_03_l {parent=index_02_l, loc=<2.5578 -0.123 -0.0252>, rot=<-0.002 0.0067 0.2558 -0.9667>}
 middle_metacarpal_l {parent=hand_l, loc=<2.1856 -0.1725 -0.5984>, rot=<0.006 0.0168 -0.0329 -0.9993>}
 middle_01_l {parent=middle_metacarpal_l, loc=<7.3275 -0.2447 0.0129>, rot=<-0.0208 0.0042 0.0858 -0.9961>}
 middle_02_l {parent=middle_01_l, loc=<4.0123 0.1401 0.1175>, rot=<-0.0218 0.0258 0.3512 -0.9357>}
 middle_03_l {parent=middle_02_l, loc=<2.6077 0.0534 0.1373>, rot=<0.0082 0.0188 0.1708 -0.9851>}
 ring_metacarpal_l {parent=hand_l, loc=<2.2074 -0.3625 1.3638>, rot=<-0.0284 0.0722 -0.0458 -0.9959>}
 ring_01_l {parent=ring_metacarpal_l, loc=<7.0295 -0.6197 -0.1446>, rot=<-0.0489 -0.0252 0.1892 -0.9804>}
 ring_02_l {parent=ring_01_l, loc=<3.6174 0.4647 -0.1027>, rot=<0 0 0.2806 -0.9598>}
 ring_03_l {parent=ring_02_l, loc=<2.4152 0.114 0.0199>, rot=<0 0 0.1189 -0.9929>}
 pinky_metacarpal_l {parent=hand_l, loc=<1.7616 -0.9753 3.3814>, rot=<-0.0397 0.0814 -0.069 -0.9935>}
 pinky_01_l {parent=pinky_metacarpal_l, loc=<7.1709 -0.4831 -0.1094>, rot=<-0.0654 -0.0035 0.2242 -0.9723>}
 pinky_02_l {parent=pinky_01_l, loc=<2.8274 0.0893 -0.1043>, rot=<0.0037 0.0011 0.2854 -0.9584>}
 pinky_03_l {parent=pinky_02_l, loc=<2.2046 0.0307 -0.0078>, rot=<-0.0002 0.0098 0.1445 -0.9895>}
 weapon_l {parent=hand_l, loc=<-0.1974 -3.4111 0.3811>, rot=<0 0 0 1>}
 hand_attach_l {parent=hand_l, loc=<8.5 -3 0>, rot=<0.7071 0.7071 0 0>}
 lowerarm_twist_02_l {parent=lowerarm_l, loc=<10.6707 0 0>, rot=<0 0 0 1>}
 lowerarm_twist_01_l {parent=lowerarm_l, loc=<18.316 0 0>, rot=<0 0 0 1>}
 deform_elbow_in_base_l {parent=lowerarm_l, loc=<0.1834 3.6062 -0.2572>, rot=<0 0 0 -1>}
 deform_elbow_in_l {parent=deform_elbow_in_base_l, loc=<0 0 0>, rot=<0 0 0 1>}
 deform_elbow_out_base_l {parent=lowerarm_l, loc=<0 -4 0>, rot=<0 0 0 1>}
 deform_elbow_out_l {parent=deform_elbow_out_base_l, loc=<0 0 0>, rot=<0 0 0 1>}
 upperarm_twist_01_l {parent=upperarm_l, loc=<7.2135 -0.1134 0>, rot=<0 0 0 1>}
 deform_upperarm_bk_base_l {parent=upperarm_twist_01_l, loc=<-6.4053 -4.3746 0.314>, rot=<0 0 0 -1>}
 deform_upperarm_bk_l {parent=deform_upperarm_bk_base_l, loc=<0 0 0>, rot=<0 0 0 1>}
 deform_upperarm_fr_base_l {parent=upperarm_twist_01_l, loc=<-5.6312 5.092 0.3354>, rot=<0 0 0 -1>}
 deform_upperarm_fr_l {parent=deform_upperarm_fr_base_l, loc=<0 0 0>, rot=<0 0 0 1>}
 deform_upperarm_up_base_l {parent=upperarm_twist_01_l, loc=<-8.891 0.4825 3.9517>, rot=<0 0 0 -1>}
 deform_upperarm_up_l {parent=deform_upperarm_up_base_l, loc=<0 0 0>, rot=<0 0 0 1>}
 upperarm_twist_02_l {parent=upperarm_l, loc=<19.7998 -0.1134 0>, rot=<0 0 0 1>}
 clavicle_r {parent=spine_05, loc=<5.4831 -0.1053 3.2946>, rot=<0.7289 0.0465 -0.6829 0.009>}
 upperarm_r {parent=clavicle_r, loc=<-9.8658 1.2136 0.4105>, rot=<0.0164 -0.3245 -0.0269 -0.9454>}
 lowerarm_r {parent=upperarm_r, loc=<-26.465 0.8345 -0.259>, rot=<0.0036 0.0161 -0.2298 -0.9731>}
 hand_r {parent=lowerarm_r, loc=<-22.7355 0.3203 0.6106>, rot=<-0.6784 -0.0223 0.0735 -0.7307>}
 thumb_01_r {parent=hand_r, loc=<-1.3118 0.3191 2.459>, rot=<0.6546 -0.4042 0.0664 -0.6354>}
 thumb_02_r {parent=thumb_01_r, loc=<-5.9336 0.0448 -0.0401>, rot=<-0.0037 0.0213 0.298 -0.9543>}
 thumb_03_r {parent=thumb_02_r, loc=<-2.2223 -0.3682 0.0907>, rot=<0.0118 -0.013 0.0066 -0.9998>}
 index_metacarpal_r {parent=hand_r, loc=<-1.6535 -0.0183 2.96>, rot=<0.0158 -0.0124 -0.0173 -0.9996>}
 index_01_r {parent=index_metacarpal_r, loc=<-8.3008 -0.0357 0.0524>, rot=<-0.0044 0.0485 0.0924 -0.9945>}
 index_02_r {parent=index_01_r, loc=<-3.3587 0.3004 -0.0452>, rot=<-0.002 0.0027 0.3225 -0.9466>}
 index_03_r {parent=index_02_r, loc=<-2.5578 0.123 0.0252>, rot=<-0.002 0.0067 0.2558 -0.9667>}
 middle_metacarpal_r {parent=hand_r, loc=<-2.1856 0.1725 0.5984>, rot=<0.006 0.0168 -0.0329 -0.9993>}
 middle_01_r {parent=middle_metacarpal_r, loc=<-7.3275 0.2447 -0.0129>, rot=<-0.0208 0.0042 0.0858 -0.9961>}
 middle_02_r {parent=middle_01_r, loc=<-4.0123 -0.1401 -0.1175>, rot=<-0.0218 0.0258 0.3512 -0.9357>}
 middle_03_r {parent=middle_02_r, loc=<-2.6077 -0.0534 -0.1373>, rot=<0.0082 0.0188 0.1708 -0.9851>}
 ring_metacarpal_r {parent=hand_r, loc=<-2.2074 0.3625 -1.3638>, rot=<-0.0284 0.0722 -0.0458 -0.9959>}
 ring_01_r {parent=ring_metacarpal_r, loc=<-7.0295 0.6197 0.1446>, rot=<-0.0489 -0.0252 0.1892 -0.9804>}
 ring_02_r {parent=ring_01_r, loc=<-3.6174 -0.4647 0.1027>, rot=<0 0 0.2806 -0.9598>}
 ring_03_r {parent=ring_02_r, loc=<-2.4152 -0.114 -0.0199>, rot=<0 0 0.1189 -0.9929>}
 pinky_metacarpal_r {parent=hand_r, loc=<-1.7616 0.9753 -3.3814>, rot=<-0.0397 0.0814 -0.069 -0.9935>}
 pinky_01_r {parent=pinky_metacarpal_r, loc=<-7.1709 0.4831 0.1094>, rot=<-0.0654 -0.0035 0.2242 -0.9723>}
 pinky_02_r {parent=pinky_01_r, loc=<-2.8274 -0.0893 0.1043>, rot=<0.0037 0.0011 0.2854 -0.9584>}
 pinky_03_r {parent=pinky_02_r, loc=<-2.2046 -0.0307 0.0078>, rot=<-0.0002 0.0098 0.1445 -0.9895>}
 weapon_r {parent=hand_r, loc=<0.1974 3.4111 -0.3811>, rot=<0 0 0 1>}
 hand_attach_r {parent=hand_r, loc=<-8.5001 2.9999 0>, rot=<0 0 -0.7071 -0.7071>}
 lowerarm_twist_02_r {parent=lowerarm_r, loc=<-10.6707 0 0>, rot=<0 0 0 -1>}
 lowerarm_twist_01_r {parent=lowerarm_r, loc=<-18.316 0 0>, rot=<0 0 0 -1>}
 mstr_braslets {parent=lowerarm_twist_01_r, loc=<0 0 0>, rot=<0 0 0 1>}
 mstr_braslet_A_move {parent=mstr_braslets, loc=<0 0 0>, rot=<0 0 0 1>}
 mstr_braslet_A {parent=mstr_braslet_A_move, loc=<-1.109 -0.2868 -3.4794>, rot=<0.914 -0.2199 0.3332 0.0722>}
 dyn_braslet_A {parent=mstr_braslet_A, loc=<0 0 0>, rot=<-0.1983 -0.8392 -0.1602 0.4804>}
 mstr_braslet_B_move {parent=mstr_braslets, loc=<0 0 0>, rot=<0 0 0 1>}
 mstr_braslet_B {parent=mstr_braslet_B_move, loc=<-3.0271 -0.2314 -3.4057>, rot=<0.914 -0.2199 0.3332 0.0722>}
 dyn_braslet_B {parent=mstr_braslet_B, loc=<0 0 0>, rot=<-0.1983 -0.8392 -0.1602 0.4804>}
 deform_elbow_in_base_r {parent=lowerarm_r, loc=<-0.1834 -3.6062 0.2572>, rot=<0 0 0 -1>}
 deform_elbow_in_r {parent=deform_elbow_in_base_r, loc=<0 0 0>, rot=<0 0 0 1>}
 deform_elbow_out_base_r {parent=lowerarm_r, loc=<0 4 -0.0001>, rot=<0 0 0 -1>}
 deform_elbow_out_r {parent=deform_elbow_out_base_r, loc=<0 0 0>, rot=<0 0 0 1>}
 upperarm_twist_01_r {parent=upperarm_r, loc=<-7.2135 0.1134 0>, rot=<0 0 0 -1>}
 deform_upperarm_bk_base_r {parent=upperarm_twist_01_r, loc=<6.4053 4.3746 -0.314>, rot=<0 0 0 -1>}
 deform_upperarm_bk_r {parent=deform_upperarm_bk_base_r, loc=<0 0 0>, rot=<0 0 0 1>}
 deform_upperarm_fr_base_r {parent=upperarm_twist_01_r, loc=<5.6312 -5.092 -0.3354>, rot=<0 0 0 -1>}
 deform_upperarm_fr_r {parent=deform_upperarm_fr_base_r, loc=<0 0 0>, rot=<0 0 0 1>}
 deform_upperarm_up_base_r {parent=upperarm_twist_01_r, loc=<8.891 -0.4825 -3.9517>, rot=<0 0 0 -1>}
 deform_upperarm_up_r {parent=deform_upperarm_up_base_r, loc=<0 0 0>, rot=<0 0 0 1>}
 upperarm_twist_02_r {parent=upperarm_r, loc=<-19.7998 0.1134 0>, rot=<0 0 0 -1>}
 attach_backpack {parent=spine_05, loc=<-0.1175 -13.4738 -0.0004>, rot=<-0.0093 -0.707 0.0092 -0.707>}
 attach_cape {parent=spine_05, loc=<9.0213 -11.7148 -0.0003>, rot=<0.0616 -0.7044 -0.0616 -0.7044>}
 dyn_hood_01 {parent=spine_05, loc=<9.1731 -8.3432 0>, rot=<0.1397 -0.6932 0.6932 -0.1397>}
 dyn_hood_02 {parent=dyn_hood_01, loc=<4.1007 0 0>, rot=<0 0 0 1>}
 mstr_lace {parent=spine_05, loc=<6.6967 8.9299 0.3961>, rot=<-0.0071 0.0114 0.0151 -0.9998>}
 dyn_lace_A_00_l {parent=mstr_lace, loc=<0.0063 -0.0357 -0.5878>, rot=<-0.3943 -0.508 0.7575 0.1123>}
 dyn_lace_A_01_l {parent=dyn_lace_A_00_l, loc=<3.2724 0.0035 -0.0208>, rot=<-0.0012 -0.0038 0.0771 -0.997>}
 dyn_lace_B_00_r {parent=mstr_lace, loc=<-0.0138 0.0367 0.5873>, rot=<0.0134 -0.7322 0.5676 0.3762>}
 dyn_lace_B_01_r {parent=dyn_lace_B_00_r, loc=<3.6868 -0.0111 -0.0144>, rot=<-0.0031 -0.0026 -0.0858 -0.9963>}
 dyn_lace_BB_00_r {parent=mstr_lace, loc=<-0.1705 0.4849 0.6456>, rot=<-0.1365 -0.7212 0.6299 0.254>}
 dyn_lace_BB_01_r {parent=dyn_lace_BB_00_r, loc=<5.3926 -0.0044 -0.0289>, rot=<0.0011 -0.0027 0.0004 -1>}
 dyn_lace_AA_00_l {parent=mstr_lace, loc=<-0.704 0.3563 -0.3772>, rot=<-0.2667 -0.5777 0.7471 0.1923>}
 dyn_lace_AA_01_l {parent=dyn_lace_AA_00_l, loc=<3.8885 -0.0009 -0.0264>, rot=<-0.0009 -0.0034 0.0001 -1>}
 dyn_hood_l_01 {parent=spine_05, loc=<11.5142 -2.1792 -8.2434>, rot=<-0.034 0.1882 0.9476 0.2558>}
 dyn_hood_r_01 {parent=spine_05, loc=<11.514 -2.179 8.243>, rot=<-0.1882 -0.034 0.2558 -0.9476>}
 dyn_hoodieTop {parent=spine_05, loc=<11.9246 -9.9059 -0.2168>, rot=<0 0 -0.036 -0.9994>}
 mstr_bag_A {parent=spine_04, loc=<-0.514 10.8129 6.0732>, rot=<-0.1069 0.2109 -0.7207 -0.6517>}
 dyn_bag_A {parent=mstr_bag_A, loc=<0 0 0>, rot=<0 0 0 1>}
 mstr_bag_B {parent=spine_04, loc=<2.5302 12.2504 -1.7975>, rot=<0.1512 0.1225 -0.7107 -0.676>}
 dyn_bag_B {parent=mstr_bag_B, loc=<0 0 0>, rot=<0 0 0 1>}
 thigh_l {parent=pelvis, loc=<-1.0855 -0.2109 -8.7053>, rot=<0.0009 0.0355 0.0155 -0.9993>}
 calf_l {parent=thigh_l, loc=<-42.5284 -0.0019 -0.7297>, rot=<0.0006 -0.0161 -0.0391 -0.9991>}
 foot_l {parent=calf_l, loc=<-46.7401 -2.5014 -0.592>, rot=<0.0427 -0.0238 0.0353 -0.9982>}
 ball_l {parent=foot_l, loc=<-3.5658 10.8066 -0.1544>, rot=<0.0001 -0.0001 -0.7063 -0.7079>}
 calf_twist_02_l {parent=calf_l, loc=<-19.1546 -1.7943 -0.2638>, rot=<0.0237 0.0095 -0.0035 -0.9997>}
 mstr_shoe_lace_l {parent=calf_twist_02_l, loc=<-5.6279 5.281 -0.259>, rot=<0 0 0 1>}
 dyn_shoe_lace_A_00_l {parent=mstr_shoe_lace_l, loc=<0.0007 -0.009 -0.3369>, rot=<-0.0502 -0.959 -0.0108 -0.2787>}
 dyn_shoe_lace_A_01_l {parent=dyn_shoe_lace_A_00_l, loc=<2.4475 0 0>, rot=<0 0 0 1>}
 dyn_shoe_lace_B_00_l {parent=mstr_shoe_lace_l, loc=<-0.0007 0.009 0.3369>, rot=<-0.0312 -0.9097 -0.0409 0.4121>}
 dyn_shoe_lace_B_01_l {parent=dyn_shoe_lace_B_00_l, loc=<1.997 0 0>, rot=<0 0 0 1>}
 calf_twist_01_l {parent=calf_l, loc=<-35.8483 -1.9197 -0.5794>, rot=<0.0237 0.0095 -0.0035 -0.9997>}
 deform_knee_in_base_l {parent=calf_l, loc=<0.8031 -5.0231 0.2456>, rot=<0 0 0 -1>}
 deform_knee_in_l {parent=deform_knee_in_base_l, loc=<0 0 0>, rot=<0 0 0 1>}
 deform_knee_out_base_l {parent=calf_l, loc=<0 4 0>, rot=<0 0 0 -1>}
 deform_knee_out_l {parent=deform_knee_out_base_l, loc=<0 0 0>, rot=<0 0 0 1>}
 thigh_twist_01_l {parent=thigh_l, loc=<-27.952 0 0>, rot=<0.0013 0 0 -1>}
 deform_groin_base_l {parent=thigh_l, loc=<0.4801 5.6776 -0.2486>, rot=<0 0 0 -1>}
 deform_groin_l {parent=deform_groin_base_l, loc=<0 0 0>, rot=<0 0 0 1>}
 thigh_r {parent=pelvis, loc=<-1.0855 -0.2109 8.7053>, rot=<0.0355 -0.0009 0.9993 0.0155>}
 calf_r {parent=thigh_r, loc=<42.5284 0.0019 0.7297>, rot=<0.0006 -0.0161 -0.0391 -0.9991>}
 foot_r {parent=calf_r, loc=<46.7401 2.5014 0.592>, rot=<0.0427 -0.0238 0.0353 -0.9982>}
 ball_r {parent=foot_r, loc=<3.5658 -10.8066 0.1544>, rot=<0.0001 -0.0001 -0.7063 -0.7079>}
 calf_twist_02_r {parent=calf_r, loc=<19.1546 1.7943 0.2638>, rot=<0.0237 0.0095 -0.0035 -0.9997>}
 mstr_shoe_lace_r {parent=calf_twist_02_r, loc=<5.628 -5.281 0.259>, rot=<0 0 0 1>}
 dyn_shoe_lace_A_00_r {parent=mstr_shoe_lace_r, loc=<-0.0007 0.009 0.3369>, rot=<0.665 0.1828 0.2543 -0.678>}
 dyn_shoe_lace_A_01_r {parent=dyn_shoe_lace_A_00_r, loc=<2.4475 0 0>, rot=<0 0 0 1>}
 dyn_shoe_lace_B_00_r {parent=mstr_shoe_lace_r, loc=<0.0007 -0.009 -0.3369>, rot=<0.6152 -0.3118 -0.2675 -0.6729>}
 dyn_shoe_lace_B_01_r {parent=dyn_shoe_lace_B_00_r, loc=<1.9969 0 0>, rot=<0 0 0 1>}
 calf_twist_01_r {parent=calf_r, loc=<35.8483 1.9197 0.5794>, rot=<0.0237 0.0095 -0.0035 -0.9997>}
 deform_knee_in_base_r {parent=calf_r, loc=<-0.8031 5.0231 -0.2457>, rot=<0 0 0 -1>}
 deform_knee_in_r {parent=deform_knee_in_base_r, loc=<0 0 0>, rot=<0 0 0 1>}
 deform_knee_out_base_r {parent=calf_r, loc=<0 -4 0>, rot=<0 0 0 -1>}
 deform_knee_out_r {parent=deform_knee_out_base_r, loc=<0 0 0>, rot=<0 0 0 1>}
 thigh_twist_01_r {parent=thigh_r, loc=<27.952 0 0>, rot=<0.0013 0 0 -1>}
 deform_groin_base_r {parent=thigh_r, loc=<-0.4801 -5.6776 0.2486>, rot=<0 0 0 -1>}
 deform_groin_r {parent=deform_groin_base_r, loc=<0 0 0>, rot=<0 0 0 1>}
 deform_glute_base_l {parent=pelvis, loc=<-7.3466 -9.0213 -9.1245>, rot=<0.0172 0.031 -0.4477 -0.8935>}
 deform_glute_l {parent=deform_glute_base_l, loc=<0 0 0>, rot=<0 0 0 1>}
 deform_glute_base_r {parent=pelvis, loc=<-7.3464 -9.0213 9.1245>, rot=<0.031 -0.0172 0.8935 -0.4477>}
 deform_glute_r {parent=deform_glute_base_r, loc=<0 0 0>, rot=<0 0 0 1>}
 deform_lower_belt_back {parent=pelvis, loc=<2.0077 -10.8921 -0.0059>, rot=<-0.0127 -0.707 0.707 0.0127>}
 deform_lower_belt_front {parent=pelvis, loc=<0.5481 9.8419 0.0575>, rot=<0.0127 0.707 0.707 0.0127>}
 deform_lower_belt_front_left {parent=pelvis, loc=<-1.4228 8.0089 -8.7531>, rot=<-0.0152 -0.8484 -0.5291 -0.0095>}
 deform_lower_belt_left {parent=pelvis, loc=<-4.0861 -0.0972 -14.8007>, rot=<-0.0179 -0.9998 0 0>}
 deform_lower_belt_back_left {parent=pelvis, loc=<-0.6948 -10.1486 -8.4037>, rot=<0.0026 -0.8472 0.5312 0.0073>}
 mstr_pistol {parent=pelvis, loc=<-5.7506 -0.4819 -13.9884>, rot=<-0.0127 -0.707 0.0127 -0.707>}
 dyn_pistol_A {parent=mstr_pistol, loc=<0 0 0>, rot=<0 0 0 1>}
 dyn_belt_mstr {parent=pelvis, loc=<5.0885 9.0566 4.3122>, rot=<-0.0468 -0.9367 -0.347 0.0043>}
 dyn_belt_base {parent=dyn_belt_mstr, loc=<0 0 0>, rot=<-0.3468 0.0105 0.03 -0.9374>}
 dyn_belt_01 {parent=dyn_belt_base, loc=<2.1633 0.3991 -2.868>, rot=<0.4145 0 0 -0.91>}
 dyn_belt_02 {parent=dyn_belt_01, loc=<2.9575 0.7488 -0.5151>, rot=<0.1318 0 0 -0.9913>}
 dyn_belt_03 {parent=dyn_belt_02, loc=<3.6692 0.4032 -0.0931>, rot=<-0.0623 0 0 -0.9981>}
 ik_hand_root {parent=root, loc=<0 0 1.3268>, rot=<0 0 0 1>}
 ik_hand_gun {parent=ik_hand_root, loc=<-48.1743 3.3545 104.4381>, rot=<0.6276 0.1309 0.4183 -0.6434>}
 ik_hand_r {parent=ik_hand_gun, loc=<0 0 0>, rot=<0 0 0 1>}
 ik_hand_l {parent=ik_hand_gun, loc=<59.3271 67.6997 34.3524>, rot=<-0.6158 -0.7027 -0.3565 0>}
 ik_foot_root {parent=root, loc=<0 0 1.3268>, rot=<0 0 0 1>}
 attach {parent=root, loc=<0 0 1.3268>, rot=<0 0 0 1>}
 root_fp {parent=root, loc=<0 0 0>, rot=<0 0 0 1>}
 attach_fp {parent=root_fp, loc=<0 0 0>, rot=<0 0 0 1>}
 camera_root {parent=root_fp, loc=<0 0 0>, rot=<0 0 0 1>}
 camera {parent=camera_root, loc=<0 0 0>, rot=<0 0 0 1>}
]
variables [isCrouching {type=bool, default=False}]
