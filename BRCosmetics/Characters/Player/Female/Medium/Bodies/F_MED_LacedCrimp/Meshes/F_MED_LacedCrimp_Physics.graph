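# Physics asset for the F_MED_LacedCrimp female player cosmetic body, set up for physics on its dangling belt bones.
# A chain of belt capsules runs from dyn_belt_05 up to dyn_belt_01, then links through sphere bodies on dyn_belt_base and dyn_belt_mstr.
# All joints lock twist, with tighter swing limits between master and base.
# Unconstrained pelvis and thigh capsules appear to serve as colliders for the belt.
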PhysicsAsset /BRCosmetics/Characters/Player/Female/Medium/Bodies/F_MED_LacedCrimp/Meshes/F_MED_LacedCrimp_Physics
bodies {
 dyn_belt_01 [{
   shapeType=Capsule
   length=3.4823
   radius=1.4185
   shapeName=dyn_belt_02_capsule
   center=<3.5451 1.0189 0>
   rotation=<rot 90 179.9999 179.9997>
  }]
 dyn_belt_02 [{shapeType=Capsule, length=3.6478, radius=1.8998, shapeName=dyn_belt_02_capsule, center=<3.4758 0 0>, rotation=<rot 90 0.0002 0>}]
 dyn_belt_03 [{shapeType=Capsule, length=3.6478, radius=1.8998, shapeName=dyn_belt_02_capsule, center=<3.5875 0 0>, rotation=<rot 90 0.0002 0>}]
 dyn_belt_04 [{shapeType=Capsule, length=3.6478, radius=1.8998, shapeName=dyn_belt_02_capsule, center=<3.5875 0 0>, rotation=<rot 90 0.0002 0>}]
 dyn_belt_05 [{shapeType=Capsule, length=3.6478, radius=1.8998, shapeName=dyn_belt_02_capsule, center=<3.5875 0 0>, rotation=<rot 90 0.0002 0>}]
 dyn_belt_base [{shapeType=Sphere, radius=1.4118, shapeName=dyn_belt_base_sphere, center=<1.1531 0.291 -1.4239>}]
 dyn_belt_mstr [{shapeType=Sphere, radius=0.505, shapeName=dyn_belt_mstr_sphere}]
 pelvis [{shapeType=Capsule, length=20.4851, radius=8.1311, shapeName=pelvis_capsule, center=<4.3 -0.6361 0>}]
 thigh_r [{
   shapeType=Capsule
   length=18.4456
   radius=9.2199
   shapeName=thigh_r_capsule
   center=<7.72 -0.0312 -2.5543>
   rotation=<rot 80.564 179.9996 179.9999>
  }]
 thigh_twist_01_l [{
   shapeType=Capsule
   length=35.4456
   radius=10.2199
   shapeName=thigh_r_capsule
   center=<9.1385 -0.0312 -0.2203>
   rotation=<rot 90 180 -179.9998>
  }]
 thigh_twist_01_r [{
   shapeType=Capsule
   length=36.4456
   radius=7.6707
   shapeName=thigh_r_capsule
   center=<-2.4249 -1.1343 -0.8004>
   rotation=<rot 84.5541 179.9991 179.9997>
  }]
}
constraints [
 {bone1Name=dyn_belt_05, bone2Name=dyn_belt_04, swing1LimitDegrees=45, swing2LimitDegrees=45, twistLimitDegrees=45, twistMotion=Locked}
 {bone1Name=dyn_belt_04, bone2Name=dyn_belt_03, swing1LimitDegrees=45, swing2LimitDegrees=45, twistLimitDegrees=45, twistMotion=Locked}
 {bone1Name=dyn_belt_03, bone2Name=dyn_belt_02, swing1LimitDegrees=45, swing2LimitDegrees=45, twistLimitDegrees=45, twistMotion=Locked}
 {bone1Name=dyn_belt_02, bone2Name=dyn_belt_01, swing1LimitDegrees=45, swing2LimitDegrees=45, twistLimitDegrees=45, twistMotion=Locked}
 {bone1Name=dyn_belt_01, bone2Name=dyn_belt_base, swing1LimitDegrees=45, swing2LimitDegrees=45, twistLimitDegrees=45, twistMotion=Locked}
 {bone1Name=dyn_belt_mstr, bone2Name=dyn_belt_base, swing1LimitDegrees=20, swing2LimitDegrees=5, twistLimitDegrees=45, twistMotion=Locked}
]
