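# Post-process style AnimBlueprint for the LacedCrimp female body cosmetic, adding secondary motion to laces, bracelets, rings and hood.
# A pose copied from the parent mesh runs through a Control Rig; a cached 'Blend FE pose' feeds the dynamics.
# bIsInFrontEnd switches between two separately tuned AnimDynamics lace chains.
# Constraints and springs then drive bracelets and rings, while bone offsets adjust the hood for jumps and arm alphas, and neck scale by LOD.
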AnimBlueprint /BRCosmetics/Characters/Player/Female/Medium/Bodies/F_MED_LacedCrimp/Meshes/F_MED_LacedCrimp_AnimBP
target_skeleton=&/F_MED_LacedCrimp_Skeleton
graphs [
 AnimGraph {nodes [
   "Transform (Modify) Bone - Bone: mstr_pistol"(Alpha=1, Translation="-0.580,-1.015,-0.000", Rotation="0,0,0", Scale="1,1,1") {
    rotationMode=BMM_Additive
    rotationSpace=BCS_BoneSpace
    translationMode=BMM_Additive
    translationSpace=BCS_BoneSpace
    alphaBoolBlend {blendOption=Linear}
    boneToModify {boneName=mstr_pistol}
    translation=<-0.58 -1.015>
   }
   n1 = Anim Dynamics - Bone: dyn_shoe_lace_A_00_l!(
    ComponentPose=Local To Component(LocalPose=Use cached pose 'Blend FE pose'())
    Alpha=1
    GravityOverride:Vector="0,0,0"
    ExternalForce="0,0,0"
   ) -> (Pose) {
    angularDampingOverride=0.999
    angularSpringConstant=2000
    bAngularSpring=true
    bOverrideAngularDamping=true
    bUseGravityOverride=true
    bUsePlanarLimit=false
    gravityScale=-1
    simulationSpace=World
    alphaBoolBlend {blendOption=Linear}
    boundBone {boneName=dyn_shoe_lace_A_00_l}
    componentAppliedLinearAccClamp=<1500 1500 1500>
    componentLinearAccScale=<0.3 0.3 0.3>
    componentLinearVelScale=<0.3 0.3 0.3>
    gravityOverride {z=-980}
    physicsBodyDefinitions [{
      sphereCollisionRadius=4
      boundBone {boneName=dyn_shoe_lace_A_00_l}
      boxExtents=<16 10 10>
      constraintSetup {angularLimitsMax=<1 10 5>, angularLimitsMin=<-1 -10 -5>, angularTarget=<1 0 0>}
      localJointOffset=<0.5 0 0>
     }]
   }
   n2 = Anim Dynamics - Bone: dyn_shoe_lace_A_01_l!(ComponentPose=@n1, Alpha=1, GravityOverride:Vector="0,0,0", ExternalForce="0,0,0") -> (Pose) {
    angularDampingOverride=0.999
    angularSpringConstant=2000
    bAngularSpring=true
    bOverrideAngularDamping=true
    bUseGravityOverride=true
    bUsePlanarLimit=false
    gravityScale=-1
    simulationSpace=World
    alphaBoolBlend {blendOption=Linear}
    boundBone {boneName=dyn_shoe_lace_A_01_l}
    componentAppliedLinearAccClamp=<1500 1500 1500>
    componentLinearAccScale=<0.3 0.3 0.3>
    componentLinearVelScale=<0.3 0.3 0.3>
    gravityOverride {z=-980}
    physicsBodyDefinitions [{
      sphereCollisionRadius=4
      boundBone {boneName=dyn_shoe_lace_A_01_l}
      boxExtents=<10 10 10>
      constraintSetup {angularLimitsMax=<1 10 5>, angularLimitsMin=<-1 -10 -5>, angularTarget=<1 0 0>}
      localJointOffset=<0.5 0 0>
     }]
   }
   n3 = Anim Dynamics - Bone: dyn_shoe_lace_B_00_l!(ComponentPose=@n2, Alpha=1, GravityOverride:Vector="0,0,0", ExternalForce="0,0,0") -> (Pose) {
    angularDampingOverride=0.999
    angularSpringConstant=1500
    bAngularSpring=true
    bOverrideAngularDamping=true
    bUseGravityOverride=true
    bUsePlanarLimit=false
    gravityScale=-1
    simulationSpace=World
    alphaBoolBlend {blendOption=Linear}
    boundBone {boneName=dyn_shoe_lace_B_00_l}
    componentAppliedLinearAccClamp=<1500 1500 1500>
    componentLinearAccScale=<0.3 0.3 0.3>
    componentLinearVelScale=<0.3 0.3 0.3>
    gravityOverride {z=-980}
    physicsBodyDefinitions [{
      sphereCollisionRadius=4
      boundBone {boneName=dyn_shoe_lace_B_00_l}
      boxExtents=<16 10 10>
      constraintSetup {angularLimitsMax=<1 10 5>, angularLimitsMin=<-1 -10 -5>, angularTarget=<1 0 0>}
      localJointOffset=<0.5 0 0>
     }]
   }
   n4 = Anim Dynamics - Bone: dyn_shoe_lace_B_01_l!(ComponentPose=@n3, Alpha=1, GravityOverride:Vector="0,0,0", ExternalForce="0,0,0") -> (Pose) {
    angularDampingOverride=0.999
    angularSpringConstant=1500
    bAngularSpring=true
    bOverrideAngularDamping=true
    bUseGravityOverride=true
    bUsePlanarLimit=false
    gravityScale=-1
    simulationSpace=World
    alphaBoolBlend {blendOption=Linear}
    boundBone {boneName=dyn_shoe_lace_B_01_l}
    componentAppliedLinearAccClamp=<1500 1500 1500>
    componentLinearAccScale=<0.3 0.3 0.3>
    componentLinearVelScale=<0.3 0.3 0.3>
    gravityOverride {z=-980}
    physicsBodyDefinitions [{
      sphereCollisionRadius=4
      boundBone {boneName=dyn_shoe_lace_B_01_l}
      boxExtents=<10 10 10>
      constraintSetup {angularLimitsMax=<1 10 5>, angularLimitsMin=<-1 -10 -5>, angularTarget=<1 0 0>}
      localJointOffset=<0.5 0 0>
     }]
   }
   n5 = Anim Dynamics - Bone: dyn_shoe_lace_A_00_r!(ComponentPose=@n4, Alpha=1, GravityOverride:Vector="0,0,0", ExternalForce="0,0,0") -> (Pose) {
    angularDampingOverride=0.9999
    angularSpringConstant=2000
    bAngularSpring=true
    bOverrideAngularDamping=true
    bUseGravityOverride=true
    bUsePlanarLimit=false
    gravityScale=-1
    simulationSpace=World
    alphaBoolBlend {blendOption=Linear}
    boundBone {boneName=dyn_shoe_lace_A_00_r}
    componentAppliedLinearAccClamp=<1000 1000 1000>
    componentLinearAccScale=<0.1 0.1 0.1>
    componentLinearVelScale=<0.2 0.2 0.2>
    gravityOverride {z=-980}
    physicsBodyDefinitions [{
      sphereCollisionRadius=4
      boundBone {boneName=dyn_shoe_lace_A_00_r}
      boxExtents=<16 10 10>
      constraintSetup {angularLimitsMax=<1 5 10>, angularLimitsMin=<-1 -5 -10>, angularTarget=<1 0 0>}
      localJointOffset=<0.5 0 0>
     }]
   }
   n6 = Anim Dynamics - Bone: dyn_shoe_lace_A_01_r!(ComponentPose=@n5, Alpha=1, GravityOverride:Vector="0,0,0", ExternalForce="0,0,0") -> (Pose) {
    angularDampingOverride=0.9999
    angularSpringConstant=2000
    bAngularSpring=true
    bOverrideAngularDamping=true
    bUseGravityOverride=true
    bUsePlanarLimit=false
    gravityScale=-1
    simulationSpace=World
    alphaBoolBlend {blendOption=Linear}
    boundBone {boneName=dyn_shoe_lace_A_01_r}
    componentAppliedLinearAccClamp=<1000 1000 1000>
    componentLinearAccScale=<0.1 0.1 0.1>
    componentLinearVelScale=<0.2 0.2 0.2>
    gravityOverride {z=-980}
    physicsBodyDefinitions [{
      sphereCollisionRadius=4
      boundBone {boneName=dyn_shoe_lace_A_01_r}
      boxExtents=<10 10 10>
      constraintSetup {angularLimitsMax=<1 5 10>, angularLimitsMin=<-1 -5 -10>, angularTarget=<1 0 0>}
      localJointOffset=<0.5 0 0>
     }]
   }
   n7 = Anim Dynamics - Bone: dyn_shoe_lace_B_00_r!(ComponentPose=@n6, Alpha=1, GravityOverride:Vector="0,0,0", ExternalForce="0,0,0") -> (Pose) {
    angularDampingOverride=1
    angularSpringConstant=2000
    bAngularSpring=true
    bOverrideAngularDamping=true
    bUseGravityOverride=true
    bUsePlanarLimit=false
    gravityScale=-1
    simulationSpace=World
    alphaBoolBlend {blendOption=Linear}
    boundBone {boneName=dyn_shoe_lace_B_00_r}
    componentAppliedLinearAccClamp=<1000 1000 1000>
    componentLinearAccScale=<0.1 0.1 0.1>
    componentLinearVelScale=<0.2 0.2 0.2>
    gravityOverride {z=-980}
    physicsBodyDefinitions [{
      sphereCollisionRadius=4
      boundBone {boneName=dyn_shoe_lace_B_00_r}
      boxExtents=<16 10 10>
      constraintSetup {angularLimitsMax=<1 5 10>, angularLimitsMin=<-1 -5 -10>, angularTarget=<1 0 0>}
      localJointOffset=<0.5 0 0>
     }]
   }
   n8 = Anim Dynamics - Bone: dyn_shoe_lace_B_01_r!(ComponentPose=@n7, Alpha=1, GravityOverride:Vector="0,0,0", ExternalForce="0,0,0") -> (Pose) {
    angularDampingOverride=1
    angularSpringConstant=2000
    bAngularSpring=true
    bOverrideAngularDamping=true
    bUseGravityOverride=true
    bUsePlanarLimit=false
    gravityScale=-1
    simulationSpace=World
    alphaBoolBlend {blendOption=Linear}
    boundBone {boneName=dyn_shoe_lace_B_01_r}
    componentAppliedLinearAccClamp=<1000 1000 1000>
    componentLinearAccScale=<0.1 0.1 0.1>
    componentLinearVelScale=<0.2 0.2 0.2>
    gravityOverride {z=-980}
    physicsBodyDefinitions [{
      sphereCollisionRadius=4
      boundBone {boneName=dyn_shoe_lace_B_01_r}
      boxExtents=<10 10 10>
      constraintSetup {angularLimitsMax=<1 5 10>, angularLimitsMin=<-1 -5 -10>, angularTarget=<1 0 0>}
      localJointOffset=<0.5 0 0>
     }]
   }
   n9 = "Constraint - mstr_lace (PP)"!(ComponentPose=@n8, Alpha=1, ConstraintWeights_0=0.9, ConstraintWeights_1:float=0.1) -> (Pose) {
    alphaBoolBlend {blendOption=Linear}
    boneToModify {boneName=mstr_lace}
    constraintSetup [{targetBone {boneName=spine_05}}, {targetBone {boneName=neck_01}}]
    constraintWeights [1,1]
   }
   n10 = Anim Dynamics - Bone: dyn_lace_A_00_l(ComponentPose=@n9, Alpha=1, ExternalForce="0,0,0") {
    angularDampingOverride=1
    angularSpringConstant=2500
    bAngularSpring=true
    bOverrideAngularDamping=true
    bUseGravityOverride=true
    simulationSpace=World
    alphaBoolBlend {blendOption=Linear}
    boundBone {boneName=dyn_lace_A_00_l}
    componentAppliedLinearAccClamp=<700 700 700>
    componentLinearAccScale=<0.1 0.1 0.1>
    componentLinearVelScale=<0.2 0.2 0.2>
    gravityOverride {z=-980}
    physicsBodyDefinitions [{
      boundBone {boneName=dyn_lace_A_00_l}
      boxExtents=<20 20 20>
      constraintSetup {angularLimitsMax=<0 5 5>, angularLimitsMin=<0 0 -5>, angularTarget=<1 0 0>}
      localJointOffset=<0.5 0 0>
     }]
   }
   n11 = Anim Dynamics - Bone: dyn_lace_A_01_l(ComponentPose=@n10, Alpha=1, ExternalForce="0,0,0") {
    angularDampingOverride=1
    angularSpringConstant=2500
    bAngularSpring=true
    bOverrideAngularDamping=true
    bUseGravityOverride=true
    simulationSpace=World
    alphaBoolBlend {blendOption=Linear}
    boundBone {boneName=dyn_lace_A_01_l}
    componentAppliedLinearAccClamp=<700 700 700>
    componentLinearAccScale=<0.1 0.1 0.1>
    componentLinearVelScale=<0.2 0.2 0.2>
    gravityOverride {z=-980}
    physicsBodyDefinitions [{
      sphereCollisionRadius=3
      boundBone {boneName=dyn_lace_A_01_l}
      boxExtents=<15 15 15>
      constraintSetup {angularLimitsMax=<0 5 10>, angularLimitsMin=<0 0 -10>, angularTarget=<1 0 0>}
      localJointOffset=<0.5 0 0>
     }]
   }
   n12 = Anim Dynamics - Bone: dyn_lace_B_00_r(ComponentPose=@n11, Alpha=1, ExternalForce="0,0,0") {
    angularDampingOverride=1
    angularSpringConstant=2500
    bAngularSpring=true
    bOverrideAngularDamping=true
    bUseGravityOverride=true
    simulationSpace=World
    alphaBoolBlend {blendOption=Linear}
    boundBone {boneName=dyn_lace_B_00_r}
    componentAppliedLinearAccClamp=<700 700 700>
    componentLinearAccScale=<0.1 0.1 0.1>
    componentLinearVelScale=<0.2 0.2 0.2>
    gravityOverride {z=-980}
    physicsBodyDefinitions [{
      boundBone {boneName=dyn_lace_B_00_r}
      boxExtents=<20 20 20>
      constraintSetup {angularLimitsMax=<0 0 5>, angularLimitsMin=<0 0 -5>, angularTarget=<1 0 0>}
      localJointOffset=<0.5 0 0>
     }]
   }
   n13 = Component To Local(ComponentPose=@n12)
   n14 = Anim Dynamics - Bone: dyn_shoe_lace_A_00_l(ComponentPose=Local To Component(LocalPose=Use cached pose 'Blend FE pose'()), Alpha=1, GravityOverride="0,0,0", ExternalForce="0,0,0") {
    angularDampingOverride=0.999
    angularSpringConstant=500
    bAngularSpring=true
    bOverrideAngularDamping=true
    bUseGravityOverride=true
    bUsePlanarLimit=false
    gravityScale=-1
    simulationSpace=RootRelative
    alphaBoolBlend {blendOption=Linear}
    boundBone {boneName=dyn_shoe_lace_A_00_l}
    componentAppliedLinearAccClamp=<3000 3000 3000>
    componentLinearAccScale=<0.3 0.3 0.3>
    componentLinearVelScale=<0.3 0.3 0.3>
    gravityOverride {z=-980}
    physicsBodyDefinitions [{
      sphereCollisionRadius=4
      boundBone {boneName=dyn_shoe_lace_A_00_l}
      boxExtents=<16 10 10>
      constraintSetup {angularLimitsMax=<0 0 5>, angularLimitsMin=<0 -45 -5>, angularTarget=<0 0 0>}
      localJointOffset=<3 0 0>
     }]
   }
   n15 = Anim Dynamics - Bone: dyn_shoe_lace_A_01_l(ComponentPose=@n14, Alpha=1, GravityOverride="0,0,0", ExternalForce="0,0,0") {
    angularDampingOverride=0.999
    angularSpringConstant=500
    bAngularSpring=true
    bOverrideAngularDamping=true
    bUseGravityOverride=true
    bUsePlanarLimit=false
    gravityScale=-1
    simulationSpace=RootRelative
    alphaBoolBlend {blendOption=Linear}
    boundBone {boneName=dyn_shoe_lace_A_01_l}
    componentAppliedLinearAccClamp=<3000 3000 3000>
    componentLinearAccScale=<0.3 0.3 0.3>
    componentLinearVelScale=<0.3 0.3 0.3>
    gravityOverride {z=-980}
    physicsBodyDefinitions [{
      sphereCollisionRadius=4
      boundBone {boneName=dyn_shoe_lace_A_01_l}
      boxExtents=<16 10 10>
      constraintSetup {angularLimitsMax=<0 0 5>, angularLimitsMin=<0 -15 -5>, angularTarget=<0 0 0>}
      localJointOffset=<3 0 0>
     }]
   }
   n16 = Anim Dynamics - Bone: dyn_shoe_lace_B_00_l(ComponentPose=@n15, Alpha=1, GravityOverride="0,0,0", ExternalForce="0,0,0") {
    angularDampingOverride=0.999
    angularSpringConstant=500
    bAngularSpring=true
    bOverrideAngularDamping=true
    bUseGravityOverride=true
    bUsePlanarLimit=false
    gravityScale=-1
    simulationSpace=RootRelative
    alphaBoolBlend {blendOption=Linear}
    boundBone {boneName=dyn_shoe_lace_B_00_l}
    componentAppliedLinearAccClamp=<3000 3000 3000>
    componentLinearAccScale=<0.3 0.3 0.3>
    componentLinearVelScale=<0.3 0.3 0.3>
    gravityOverride {z=-980}
    physicsBodyDefinitions [{
      sphereCollisionRadius=4
      boundBone {boneName=dyn_shoe_lace_B_00_l}
      boxExtents=<16 10 10>
      constraintSetup {angularLimitsMax=<0 45 5>, angularLimitsMin=<0 0 -5>, angularTarget=<0 0 0>}
      localJointOffset=<3 0 0>
     }]
   }
   n17 = Anim Dynamics - Bone: dyn_shoe_lace_B_01_l(ComponentPose=@n16, Alpha=1, GravityOverride="0,0,0", ExternalForce="0,0,0") {
    angularDampingOverride=0.999
    angularSpringConstant=500
    bAngularSpring=true
    bOverrideAngularDamping=true
    bUseGravityOverride=true
    bUsePlanarLimit=false
    gravityScale=-1
    simulationSpace=RootRelative
    alphaBoolBlend {blendOption=Linear}
    boundBone {boneName=dyn_shoe_lace_B_01_l}
    componentAppliedLinearAccClamp=<3000 3000 3000>
    componentLinearAccScale=<0.3 0.3 0.3>
    componentLinearVelScale=<0.3 0.3 0.3>
    gravityOverride {z=-980}
    physicsBodyDefinitions [{
      sphereCollisionRadius=4
      boundBone {boneName=dyn_shoe_lace_B_01_l}
      boxExtents=<16 10 10>
      constraintSetup {angularLimitsMax=<0 0 5>, angularLimitsMin=<0 -45 -5>, angularTarget=<0 0 0>}
      localJointOffset=<3 0 0>
     }]
   }
   n18 = Anim Dynamics - Bone: dyn_shoe_lace_A_00_r(ComponentPose=@n17, Alpha=1, GravityOverride="0,0,0", ExternalForce="0,0,0") {
    angularDampingOverride=0.999
    angularSpringConstant=500
    bAngularSpring=true
    bOverrideAngularDamping=true
    bUseGravityOverride=true
    bUsePlanarLimit=false
    gravityScale=-1
    simulationSpace=RootRelative
    alphaBoolBlend {blendOption=Linear}
    boundBone {boneName=dyn_shoe_lace_A_00_r}
    componentAppliedLinearAccClamp=<3000 3000 3000>
    componentLinearAccScale=<0.3 0.3 0.3>
    componentLinearVelScale=<0.3 0.3 0.3>
    gravityOverride {z=-980}
    physicsBodyDefinitions [{
      sphereCollisionRadius=4
      boundBone {boneName=dyn_shoe_lace_A_00_r}
      boxExtents=<16 10 10>
      constraintSetup {angularLimitsMax=<0 0 5>, angularLimitsMin=<0 -45 -5>, angularTarget=<0 0 0>}
      localJointOffset=<3 0 0>
     }]
   }
   n19 = Anim Dynamics - Bone: dyn_shoe_lace_A_01_r(ComponentPose=@n18, Alpha=1, GravityOverride="0,0,0", ExternalForce="0,0,0") {
    angularDampingOverride=0.999
    angularSpringConstant=500
    bAngularSpring=true
    bOverrideAngularDamping=true
    bUseGravityOverride=true
    bUsePlanarLimit=false
    gravityScale=-1
    simulationSpace=RootRelative
    alphaBoolBlend {blendOption=Linear}
    boundBone {boneName=dyn_shoe_lace_A_01_r}
    componentAppliedLinearAccClamp=<3000 3000 3000>
    componentLinearAccScale=<0.3 0.3 0.3>
    componentLinearVelScale=<0.3 0.3 0.3>
    gravityOverride {z=-980}
    physicsBodyDefinitions [{
      sphereCollisionRadius=4
      boundBone {boneName=dyn_shoe_lace_A_01_r}
      boxExtents=<16 10 10>
      constraintSetup {angularLimitsMax=<0 0 5>, angularLimitsMin=<0 -45 -5>, angularTarget=<0 0 0>}
      localJointOffset=<3 0 0>
     }]
   }
   n20 = Anim Dynamics - Bone: dyn_shoe_lace_B_00_r(ComponentPose=@n19, Alpha=1, GravityOverride="0,0,0", ExternalForce="0,0,0") {
    angularDampingOverride=0.999
    angularSpringConstant=500
    bAngularSpring=true
    bOverrideAngularDamping=true
    bUseGravityOverride=true
    bUsePlanarLimit=false
    gravityScale=-1
    simulationSpace=RootRelative
    alphaBoolBlend {blendOption=Linear}
    boundBone {boneName=dyn_shoe_lace_B_00_r}
    componentAppliedLinearAccClamp=<3000 3000 3000>
    componentLinearAccScale=<0.3 0.3 0.3>
    componentLinearVelScale=<0.3 0.3 0.3>
    gravityOverride {z=-980}
    physicsBodyDefinitions [{
      sphereCollisionRadius=4
      boundBone {boneName=dyn_shoe_lace_B_00_r}
      boxExtents=<16 10 10>
      constraintSetup {angularLimitsMax=<0 0 5>, angularLimitsMin=<0 -45 -5>, angularTarget=<0 0 0>}
      localJointOffset=<3 0 0>
     }]
   }
   n21 = Anim Dynamics - Bone: dyn_shoe_lace_B_01_r(ComponentPose=@n20, Alpha=1, GravityOverride="0,0,0", ExternalForce="0,0,0") {
    angularDampingOverride=0.999
    angularSpringConstant=500
    bAngularSpring=true
    bOverrideAngularDamping=true
    bUseGravityOverride=true
    bUsePlanarLimit=false
    gravityScale=-1
    simulationSpace=RootRelative
    alphaBoolBlend {blendOption=Linear}
    boundBone {boneName=dyn_shoe_lace_B_01_r}
    componentAppliedLinearAccClamp=<3000 3000 3000>
    componentLinearAccScale=<0.3 0.3 0.3>
    componentLinearVelScale=<0.3 0.3 0.3>
    gravityOverride {z=-980}
    physicsBodyDefinitions [{
      sphereCollisionRadius=4
      boundBone {boneName=dyn_shoe_lace_B_01_r}
      boxExtents=<16 10 10>
      constraintSetup {angularLimitsMax=<0 0 5>, angularLimitsMin=<0 -45 -5>, angularTarget=<0 0 0>}
      localJointOffset=<3 0 0>
     }]
   }
   n22 = "Constraint - mstr_lace (PP)"(ComponentPose=@n21, Alpha=1, ConstraintWeights_0=0.9, ConstraintWeights_1=0.1) {
    alphaBoolBlend {blendOption=Linear}
    boneToModify {boneName=mstr_lace}
    constraintSetup [{targetBone {boneName=spine_05}}, {targetBone {boneName=neck_01}}]
    constraintWeights [1,1]
   }
   n23 = Anim Dynamics - Bone: dyn_lace_A_00_l(ComponentPose=@n22, Alpha=1, ExternalForce="0,0,0") {
    angularDampingOverride=1
    bLinearSpring=true
    bOverrideAngularDamping=true
    bUseGravityOverride=true
    linearSpringConstant=500
    simulationSpace=RootRelative
    alphaBoolBlend {blendOption=Linear}
    boundBone {boneName=dyn_lace_A_00_l}
    componentAppliedLinearAccClamp=<3000 3000 3000>
    componentLinearAccScale=<0.3 0.3 0.3>
    componentLinearVelScale=<0.3 0.3 0.3>
    gravityOverride {z=-980}
    physicsBodyDefinitions [{
      boundBone {boneName=dyn_lace_A_00_l}
      boxExtents=<16 10 10>
      constraintSetup {angularLimitsMax=<10 0 15>, angularLimitsMin=<-10 0 -5>, angularTarget=<1 0 0>}
      localJointOffset=<3 0 0>
     }]
   }
   n24 = Anim Dynamics - Bone: dyn_lace_A_01_l(ComponentPose=@n23, Alpha=1, ExternalForce="0,0,0") {
    angularDampingOverride=1
    bLinearSpring=true
    bOverrideAngularDamping=true
    bUseGravityOverride=true
    linearSpringConstant=500
    alphaBoolBlend {blendOption=Linear}
    boundBone {boneName=dyn_lace_A_01_l}
    componentAppliedLinearAccClamp=<3000 3000 3000>
    componentLinearAccScale=<0.3 0.3 0.3>
    componentLinearVelScale=<0.3 0.3 0.3>
    gravityOverride {z=-980}
    physicsBodyDefinitions [{
      sphereCollisionRadius=3
      boundBone {boneName=dyn_lace_A_01_l}
      boxExtents=<16 10 10>
      constraintSetup {angularLimitsMax=<8 5 10>, angularLimitsMin=<-8 0 -10>, angularTarget=<1 0 0>}
      localJointOffset=<3 0 0>
     }]
   }
   n25 = Anim Dynamics - Bone: dyn_lace_B_00_r(ComponentPose=@n24, Alpha=1, ExternalForce="0,0,0") {
    angularDampingOverride=1
    bLinearSpring=true
    bOverrideAngularDamping=true
    bUseGravityOverride=true
    linearSpringConstant=500
    simulationSpace=RootRelative
    alphaBoolBlend {blendOption=Linear}
    boundBone {boneName=dyn_lace_B_00_r}
    componentAppliedLinearAccClamp=<3000 3000 3000>
    componentLinearAccScale=<0.3 0.3 0.3>
    componentLinearVelScale=<0.3 0.3 0.3>
    gravityOverride {z=-980}
    physicsBodyDefinitions [{
      boundBone {boneName=dyn_lace_B_00_r}
      boxExtents=<16 10 10>
      constraintSetup {angularLimitsMax=<10 0 5>, angularLimitsMin=<-10 0 -15>, angularTarget=<1 0 0>}
      localJointOffset=<3 0 0>
     }]
   }
   n26 = Anim Dynamics - Bone: dyn_lace_B_01_r(ComponentPose=@n25, Alpha=1, ExternalForce="0,0,0") {
    angularDampingOverride=1
    bOverrideAngularDamping=true
    bUseGravityOverride=true
    linearSpringConstant=500
    alphaBoolBlend {blendOption=Linear}
    boundBone {boneName=dyn_lace_B_01_r}
    componentAppliedLinearAccClamp=<3000 3000 3000>
    componentLinearAccScale=<0.3 0.3 0.3>
    componentLinearVelScale=<0.3 0.3 0.3>
    gravityOverride {z=-980}
    physicsBodyDefinitions [{
      sphereCollisionRadius=3
      boundBone {boneName=dyn_lace_B_01_r}
      boxExtents=<16 10 10>
      constraintSetup {angularLimitsMax=<8 5 10>, angularLimitsMin=<-8 0 -10>, angularTarget=<1 0 0>}
      localJointOffset=<3 0 0>
     }]
   }
   n27 = Component To Local(ComponentPose=@n26)
   n28 = Get bIsInFrontEnd()
   n29 = "Constraint - mstr_braslet_A (PPP)"!(
    ComponentPose=Local To Component(LocalPose=Blend Poses by bool(BlendPose_0=@n13, BlendPose_1=@n27, bActiveValue=@n28) {blendType=Linear})
    Alpha=1
    ConstraintWeights_0=0.06
    ConstraintWeights_1:float=0.0859
    ConstraintWeights_2:float=0.8541
   ) -> (Pose) {
    alphaBoolBlend {blendOption=Linear}
    boneToModify {boneName=mstr_braslet_A}
    constraintSetup [{targetBone {boneName=dyn_braslet_A}}, {targetBone {boneName=lowerarm_r}}, {targetBone {boneName=lowerarm_twist_01_r}}]
    constraintWeights [1,1,1]
   }
   n30 = Anim Dynamics - Bone: mstr_braslet_A(ComponentPose=@n29, Alpha=1, ExternalForce="0,0,0") {
    angularDampingOverride=0.96
    angularSpringConstant=75
    bAngularSpring=true
    bGravityOverrideInSimSpace=true
    bLinearSpring=true
    bOverrideAngularDamping=true
    bOverrideLinearDamping=true
    simulationSpace=Actor
    alphaBoolBlend {blendOption=Linear}
    boundBone {boneName=mstr_braslet_A}
    componentAppliedLinearAccClamp=<4000 4000 4000>
    componentLinearAccScale=<1 1 1>
    componentLinearVelScale=<0.5 0.5 0.5>
    physicsBodyDefinitions [{
      sphereCollisionRadius=0
      boundBone {boneName=mstr_braslet_A}
      boxExtents=<5 18 18>
      constraintSetup {angularLimitsMax=<15 20 0>, angularLimitsMin=<-10 -20 0>, angularTarget=<1 0 0>}
      localJointOffset=<2 0 -4>
     }]
   }
   n31 = "Constraint - mstr_braslet_B (PPP)"!(ComponentPose=@n30, Alpha=1, ConstraintWeights_0=0.06, ConstraintWeights_1:float=0.0859, ConstraintWeights_2:float=0.8541) -> (Pose) {
    alphaBoolBlend {blendOption=Linear}
    boneToModify {boneName=mstr_braslet_B}
    constraintSetup [{targetBone {boneName=mstr_braslet_B}}, {targetBone {boneName=lowerarm_r}}, {targetBone {boneName=lowerarm_twist_01_r}}]
    constraintWeights [1,1,1]
   }
   n32 = Anim Dynamics - Bone: mstr_braslet_B(ComponentPose=@n31, Alpha=1, ExternalForce="0,0,0") {
    angularDampingOverride=0.96
    angularSpringConstant=75
    bAngularSpring=true
    bGravityOverrideInSimSpace=true
    bLinearSpring=true
    bOverrideAngularDamping=true
    bOverrideLinearDamping=true
    simulationSpace=Actor
    alphaBoolBlend {blendOption=Linear}
    boundBone {boneName=mstr_braslet_B}
    componentAppliedLinearAccClamp=<4000 4000 4000>
    componentLinearAccScale=<1 1 1>
    componentLinearVelScale=<0.5 0.5 0.5>
    physicsBodyDefinitions [{
      sphereCollisionRadius=0
      boundBone {boneName=mstr_braslet_B}
      boxExtents=<5 18 18>
      constraintSetup {angularLimitsMax=<15 20 0>, angularLimitsMin=<-10 -20 0>, angularTarget=<1 0 0>}
      localJointOffset=<2 0 -4>
     }]
   }
   n33 = "Transform (Modify) Bone - Bone: mstr_braslet_A"(ComponentPose=@n32, Alpha=0, Translation="0.335,0,0.335", Rotation="0,0,0", Scale="1,1,1") {
    rotationMode=BMM_Additive
    rotationSpace=BCS_BoneSpace
    translationMode=BMM_Additive
    translationSpace=BCS_BoneSpace
    alphaBoolBlend {blendOption=Linear}
    boneToModify {boneName=mstr_braslet_A}
    translation {x=0.335, z=0.335}
   }
   n34 = "Transform (Modify) Bone - Bone: mstr_braslet_B"(ComponentPose=@n33, Alpha=0, Translation="0.335,0,0.335", Rotation="0,0,0", Scale="1,1,1") {
    rotationMode=BMM_Additive
    rotationSpace=BCS_BoneSpace
    translationMode=BMM_Additive
    translationSpace=BCS_BoneSpace
    alphaBoolBlend {blendOption=Linear}
    boneToModify {boneName=mstr_braslet_B}
    translation {x=0.335, z=0.335}
   }
   n35 = "Transform (Modify) Bone - Bone: mstr_braslets"(ComponentPose=@n34, Alpha=1, Translation="1.193,-0.473,0.009", Rotation="0,0,0", Scale="1,1,1") {
    rotationMode=BMM_Additive
    rotationSpace=BCS_BoneSpace
    translationMode=BMM_Additive
    translationSpace=BCS_BoneSpace
    alphaBoolBlend {blendOption=Linear}
    boneToModify {boneName=mstr_braslets}
    translation=<1.193 -0.473 0.009>
   }
   n36 = Anim Dynamics - Bone: mstr_braslet_A_move(ComponentPose=@n35, Alpha=1, ExternalForce="0,0,0") {
    bLinearSpring=true
    bOverrideLinearDamping=true
    bUseGravityOverride=true
    linearDampingOverride=1
    linearSpringConstant=175
    simulationSpace=RootRelative
    alphaBoolBlend {blendOption=Linear}
    boundBone {boneName=mstr_braslet_A_move}
    componentAppliedLinearAccClamp=<2500 2500 2500>
    componentLinearAccScale=<0.2 0.2 0.2>
    componentLinearVelScale=<0.9 0.9 0.9>
    physicsBodyDefinitions [{
      boundBone {boneName=mstr_braslet_A_move}
      boxExtents=<60 60 60>
      constraintSetup {angularLimitsMax=<0 0 0>, angularLimitsMin=<0 0 0>, angularTarget=<0 0 0>, linearAxesMax=<0.2 0 0>, linearAxesMin=<-0.2 0 0>}
      localJointOffset=<0 0 0>
     }]
   }
   n37 = Anim Dynamics - Bone: mstr_braslet_B_move(ComponentPose=@n36, Alpha=0.6, ExternalForce="0,0,0") {
    bLinearSpring=true
    bOverrideLinearDamping=true
    bUseGravityOverride=true
    linearDampingOverride=1
    linearSpringConstant=175
    simulationSpace=RootRelative
    alphaBoolBlend {blendOption=Linear}
    boundBone {boneName=mstr_braslet_B_move}
    componentAppliedLinearAccClamp=<2500 2500 2500>
    componentLinearAccScale=<0.2 0.2 0.2>
    componentLinearVelScale=<0.9 0.9 0.9>
    physicsBodyDefinitions [{
      boundBone {boneName=mstr_braslet_B_move}
      boxExtents=<60 60 60>
      constraintSetup {angularLimitsMax=<0 0 0>, angularLimitsMin=<0 0 0>, angularTarget=<0 0 0>, linearAxesMax=<0.2 0 0>, linearAxesMin=<-0.2 0 0>}
      localJointOffset=<0 0 0>
     }]
   }
   n38 = Anim Dynamics - Bone: dyn_ring_A(ComponentPose=@n37, Alpha=1, ExternalForce="0,0,0") {
    angularDampingOverride=0.96
    angularSpringConstant=75
    bAngularSpring=true
    bGravityOverrideInSimSpace=true
    bLinearSpring=true
    bOverrideAngularDamping=true
    bOverrideLinearDamping=true
    simulationSpace=Actor
    alphaBoolBlend {blendOption=Linear}
    boundBone {boneName=dyn_ring_A}
    componentAppliedLinearAccClamp=<4000 4000 4000>
    componentLinearAccScale=<1 1 1>
    componentLinearVelScale=<0.5 0.5 0.5>
    physicsBodyDefinitions [{
      sphereCollisionRadius=0
      boundBone {boneName=dyn_ring_A}
      boxExtents=<5 18 18>
      constraintSetup {angularLimitsMax=<15 20 0>, angularLimitsMin=<-10 -20 0>, angularTarget=<1 0 0>}
      localJointOffset=<2 0 -4>
     }]
   }
   n39 = Anim Dynamics - Bone: dyn_ring_B(ComponentPose=@n38, Alpha=1, ExternalForce="0,0,0") {
    angularDampingOverride=0.96
    angularSpringConstant=75
    bAngularSpring=true
    bGravityOverrideInSimSpace=true
    bLinearSpring=true
    bOverrideAngularDamping=true
    bOverrideLinearDamping=true
    simulationSpace=Actor
    alphaBoolBlend {blendOption=Linear}
    boundBone {boneName=dyn_ring_B}
    componentAppliedLinearAccClamp=<4000 4000 4000>
    componentLinearAccScale=<1 1 1>
    componentLinearVelScale=<0.5 0.5 0.5>
    physicsBodyDefinitions [{
      sphereCollisionRadius=0
      boundBone {boneName=dyn_ring_B}
      boxExtents=<5 18 18>
      constraintSetup {angularLimitsMax=<15 20 0>, angularLimitsMin=<-10 -20 0>, angularTarget=<1 0 0>}
      localJointOffset=<2 0 -4>
     }]
   }
   n40 = Anim Dynamics - Bone: dyn_ring_C(ComponentPose=@n39, Alpha=1, ExternalForce="0,0,0") {
    angularDampingOverride=0.96
    angularSpringConstant=75
    bAngularSpring=true
    bGravityOverrideInSimSpace=true
    bLinearSpring=true
    bOverrideAngularDamping=true
    bOverrideLinearDamping=true
    simulationSpace=Actor
    alphaBoolBlend {blendOption=Linear}
    boundBone {boneName=dyn_ring_C}
    componentAppliedLinearAccClamp=<4000 4000 4000>
    componentLinearAccScale=<1 1 1>
    componentLinearVelScale=<0.5 0.5 0.5>
    physicsBodyDefinitions [{
      sphereCollisionRadius=0
      boundBone {boneName=dyn_ring_C}
      boxExtents=<5 18 18>
      constraintSetup {angularLimitsMax=<15 20 0>, angularLimitsMin=<-10 -20 0>, angularTarget=<1 0 0>}
      localJointOffset=<2 0 -4>
     }]
   }
   n41 = Get is in jump or falling() -> (is in jump or falling:bool)
   n42 = Get jump easing tr() -> (jump easing tr:double)
   n43 = "Transform (Modify) Bone - Bone: dyn_hood_01"!(
    ComponentPose=@n40
    bAlphaBoolEnabled:bool=@n41
    Translation_X:double=0
    Translation_Y:double=0
    Translation_Z:double=@n42
    Rotation_Roll:float=0
    Rotation_Pitch:float=0
    Rotation_Yaw:float=Get jump easing() -> (jump easing:double)
    Scale="1,1,1"
   ) -> (Pose) {
    alphaInputType=Bool
    rotationMode=BMM_Additive
    rotationSpace=BCS_BoneSpace
    translationMode=BMM_Additive
    translationSpace=BCS_WorldSpace
    alphaBoolBlend {blendInTime=0.4, blendOption=Linear, blendOutTime=0.4}
    boneToModify {boneName=dyn_hood_01}
   }
   n44 = Get side jump easing() -> (side jump easing:double)
   n45 = "Transform (Modify) Bone - Bone: dyn_hood_l_01"!(
    ComponentPose=@n43
    bAlphaBoolEnabled:bool=@n41
    Translation_X:double=0
    Translation_Y:double=0
    Translation_Z:double=@n42
    Rotation_Roll:float=0
    Rotation_Pitch:float=0
    Rotation_Yaw:float=@n44
    Scale="1,1,1"
   ) -> (Pose) {
    alphaInputType=Bool
    rotationMode=BMM_Additive
    rotationSpace=BCS_BoneSpace
    translationMode=BMM_Additive
    translationSpace=BCS_WorldSpace
    alphaBoolBlend {blendInTime=0.4, blendOption=Linear, blendOutTime=0.4}
    boneToModify {boneName=dyn_hood_l_01}
   }
   n46 = "Transform (Modify) Bone - Bone: dyn_hood_r_01"!(
    ComponentPose=@n45
    bAlphaBoolEnabled:bool=@n41
    Translation_X:double=0
    Translation_Y:double=0
    Translation_Z:double=@n42
    Rotation_Roll:float=0
    Rotation_Pitch:float=0
    Rotation_Yaw:float=@n44
    Scale="1,1,1"
   ) -> (Pose) {
    alphaInputType=Bool
    rotationMode=BMM_Additive
    rotationSpace=BCS_BoneSpace
    translationSpace=BCS_WorldSpace
    alphaBoolBlend {blendInTime=0.4, blendOption=Linear, blendOutTime=0.4}
    boneToModify {boneName=dyn_hood_r_01}
   }
   n47 = "Transform (Modify) Bone - Bone: dyn_hood_01"!(ComponentPose=@n46, Alpha:float=1, Translation:Vector="-5,-2,0", Rotation:Rotator="0,0,0", Scale="1,1,1") -> (Pose) {
    rotationMode=BMM_Additive
    rotationSpace=BCS_BoneSpace
    translationMode=BMM_Additive
    translationSpace=BCS_BoneSpace
    alphaBoolBlend {blendOption=Linear}
    boneToModify {boneName=dyn_hood_01}
    translation=<-5 -2>
   }
   n48 = "Transform (Modify) Bone - Bone: dyn_hood_l_01"!(ComponentPose=@n47, Alpha:float=1, Translation:Vector="3,1,0", Rotation:Rotator="0,0,0", Scale="1,1,1") -> (Pose) {
    rotationMode=BMM_Additive
    rotationSpace=BCS_BoneSpace
    translationMode=BMM_Additive
    translationSpace=BCS_BoneSpace
    alphaBoolBlend {blendOption=Linear}
    boneToModify {boneName=dyn_hood_l_01}
    translation=<3 1>
   }
   n49 = "Transform (Modify) Bone - Bone: dyn_hood_r_01"!(ComponentPose=@n48, Alpha:float=1, Translation:Vector="-3,-1,0", Rotation:Rotator="0,0,0", Scale="1,1,1") -> (Pose) {
    rotationMode=BMM_Additive
    rotationSpace=BCS_BoneSpace
    translationMode=BMM_Additive
    translationSpace=BCS_BoneSpace
    alphaBoolBlend {blendOption=Linear}
    boneToModify {boneName=dyn_hood_r_01}
    translation=<-3 -1>
   }
   n50 = "Transform (Modify) Bone - Bone: dyn_hoodieTop"!(ComponentPose=@n49, bAlphaBoolEnabled:bool=True, Translation="-2.5,-1.468,0.353", Rotation="0,0,0", Scale="1,1,1") -> (Pose) {
    alphaInputType=Bool
    rotationMode=BMM_Additive
    rotationSpace=BCS_BoneSpace
    scaleMode=BMM_Additive
    scaleSpace=BCS_BoneSpace
    translationMode=BMM_Additive
    translationSpace=BCS_BoneSpace
    alphaBoolBlend {blendInTime=0.1, blendOption=Linear, blendOutTime=0.1}
    boneToModify {boneName=dyn_hoodieTop}
    translation=<-2.5 -1.468 0.353>
   }
   n51 = "Transform (Modify) Bone - Bone: dyn_hoodieTop"(ComponentPose=@n50, bAlphaBoolEnabled=True, Translation="4,1.468,-0.353", Rotation="0,0,0", Scale="1,1,1") {
    alphaInputType=Bool
    rotationMode=BMM_Additive
    rotationSpace=BCS_BoneSpace
    scaleMode=BMM_Additive
    scaleSpace=BCS_BoneSpace
    translationMode=BMM_Additive
    translationSpace=BCS_BoneSpace
    alphaBoolBlend {blendInTime=0.1, blendOption=Linear, blendOutTime=0.1}
    boneToModify {boneName=dyn_hoodieTop}
    translation=<4 1.468 -0.353>
   }
   n52 = Anim Dynamics - Bone: dyn_hood_01!(ComponentPose=@n51, Alpha=0.85, GravityOverride:Vector="0,0,0", ExternalForce="0,0,0") -> (Pose) {
    angularDampingOverride=0.9999
    angularSpringConstant=75
    bAngularSpring=true
    bGravityOverrideInSimSpace=true
    bOverrideAngularDamping=true
    simulationSpace=RootRelative
    alphaBoolBlend {blendOption=Linear}
    boundBone {boneName=dyn_hood_01}
    componentAppliedLinearAccClamp=<5000 5000 5000>
    componentLinearAccScale=<1 1 1>
    componentLinearVelScale=<1 1 1>
    physicsBodyDefinitions [{
      sphereCollisionRadius=0
      boundBone {boneName=dyn_hood_01}
      boxExtents=<10 20 20>
      constraintSetup {angularLimitsMax=<0 0 0>, angularLimitsMin=<0 -70 0>, angularTarget=<0 0 0>}
      localJointOffset=<2 0 0>
     }]
   }
   n53 = Anim Dynamics - Bone: dyn_hood_02!(ComponentPose=@n52, Alpha=0.85, GravityOverride:Vector="0,0,0", ExternalForce="0,0,0") -> (Pose) {
    angularDampingOverride=0.9999
    angularSpringConstant=75
    bAngularSpring=true
    bGravityOverrideInSimSpace=true
    bOverrideAngularDamping=true
    simulationSpace=RootRelative
    alphaBoolBlend {blendOption=Linear}
    boundBone {boneName=dyn_hood_02}
    componentAppliedLinearAccClamp=<5000 5000 5000>
    componentLinearAccScale=<1 1 1>
    componentLinearVelScale=<1 1 1>
    physicsBodyDefinitions [{
      sphereCollisionRadius=0
      boundBone {boneName=dyn_hood_02}
      boxExtents=<10 20 20>
      constraintSetup {angularLimitsMax=<0 0 0>, angularLimitsMin=<0 -50 0>, angularTarget=<0 0 0>}
      localJointOffset=<2 0 0>
     }]
   }
   n54 = "Transform (Modify) Bone - Bone: dyn_hood_l_01"(
    ComponentPose=@n53
    Alpha=Get Alpha upperarm_l() -> (Alpha upperarm_l:float)
    Translation="-3.914,-1.865,2.504"
    Rotation="0,0,0"
    Scale="1.000,1.000,0.097"
   ) {
    rotationMode=BMM_Additive
    rotationSpace=BCS_BoneSpace
    scaleMode=BMM_Additive
    scaleSpace=BCS_BoneSpace
    translationMode=BMM_Additive
    translationSpace=BCS_BoneSpace
    alphaBoolBlend {blendOption=Linear}
    boneToModify {boneName=dyn_hood_l_01}
    scale {z=0.097}
    translation=<-3.914 -1.865 2.504>
   }
   n55 = "Transform (Modify) Bone - Bone: dyn_hood_r_01"(
    ComponentPose=@n54
    Alpha=Get Alpha upperarm_r() -> (Alpha upperarm_r:float)
    Translation="3.914,1.865,-2.504"
    Rotation="0,0,0"
    Scale="1.000,1.000,0.097"
   ) {
    rotationMode=BMM_Additive
    rotationSpace=BCS_BoneSpace
    scaleMode=BMM_Additive
    scaleSpace=BCS_BoneSpace
    translationMode=BMM_Additive
    translationSpace=BCS_BoneSpace
    alphaBoolBlend {blendOption=Linear}
    boneToModify {boneName=dyn_hood_r_01}
    scale {z=0.097}
    translation=<3.914 1.865 -2.504>
   }
   n56 = Get 1-2-3 LODs alpha() -> (1-2-3 LODs alpha:bool)
   n57 = "Transform (Modify) Bone - Bone: dyn_neck_01"!(ComponentPose=@n55, bAlphaBoolEnabled:bool=@n56, Translation="0,0,0", Rotation="0,0,0", Scale="1.15,1.15,1.15") -> (Pose) {
    alphaInputType=Bool
    scaleMode=BMM_Additive
    scaleSpace=BCS_BoneSpace
    alphaBoolBlend {blendInTime=0.2, blendOption=Linear, blendOutTime=0.2}
    boneToModify {boneName=dyn_neck_01}
    scale=<1.15 1.15 1.15>
   }
   n58 = "Transform (Modify) Bone - Bone: dyn_neck_02"!(ComponentPose=@n57, bAlphaBoolEnabled:bool=@n56, Translation="0,0,0", Rotation="0,0,0", Scale="1.15,1.15,1.15") -> (Pose) {
    alphaInputType=Bool
    scaleMode=BMM_Additive
    scaleSpace=BCS_BoneSpace
    alphaBoolBlend {blendInTime=0.2, blendOption=Linear, blendOutTime=0.2}
    boneToModify {boneName=dyn_neck_02}
    scale=<1.15 1.15 1.15>
   }
   Output Pose(Result=Component To Local(ComponentPose=@n58)) {layerGroup=None}
   n59 = Control Rig(Source=Copy Pose From Mesh() {bUseAttachedParent=true}) {
    alphaBoolBlend {blendOption=Linear}
    controlRigAssetReference {blueprintRigClass=&/F_MED_LacedCrimp_CtrlRig.F_MED_LacedCrimp_CtrlRig_C}
   }
   n60 = "Transform (Modify) Bone - Bone: FACIAL_C_UnderChin"!(ComponentPose=Local To Component(LocalPose=@n59), bAlphaBoolEnabled:bool=True, Translation="0,0,0", Rotation="0,0,20") -> (Pose) {
    alphaInputType=Bool
    rotationMode=BMM_Replace
    rotationSpace=BCS_ParentBoneSpace
    translationMode=BMM_Replace
    translationSpace=BCS_ParentBoneSpace
    alphaBoolBlend {blendOption=Linear}
    boneToModify {boneName=FACIAL_C_UnderChin}
    rotation {roll=20}
   }
   n61 = "Transform (Modify) Bone - Bone: FACIAL_L_UnderChin"!(ComponentPose=@n60, bAlphaBoolEnabled:bool=True, Translation="0,0,0", Rotation="0,0,20") -> (Pose) {
    alphaInputType=Bool
    rotationMode=BMM_Replace
    rotationSpace=BCS_ParentBoneSpace
    translationMode=BMM_Replace
    translationSpace=BCS_ParentBoneSpace
    alphaBoolBlend {blendOption=Linear}
    boneToModify {boneName=FACIAL_L_UnderChin}
    rotation {roll=20}
   }
   n62 = "Transform (Modify) Bone - Bone: FACIAL_R_UnderChin"!(ComponentPose=@n61, bAlphaBoolEnabled:bool=True, Translation="0,0,0", Rotation="0,0,20") -> (Pose) {
    alphaInputType=Bool
    rotationMode=BMM_Replace
    rotationSpace=BCS_ParentBoneSpace
    translationMode=BMM_Replace
    translationSpace=BCS_ParentBoneSpace
    alphaBoolBlend {blendOption=Linear}
    boneToModify {boneName=FACIAL_R_UnderChin}
    rotation {roll=20}
   }
   n63 = "Transform (Modify) Bone - Bone: FACIAL_L_Jawline"!(ComponentPose=@n62, bAlphaBoolEnabled:bool=True, Translation="0,0,0", Rotation="0,0,20") -> (Pose) {
    alphaInputType=Bool
    rotationMode=BMM_Replace
    rotationSpace=BCS_ParentBoneSpace
    translationMode=BMM_Replace
    translationSpace=BCS_ParentBoneSpace
    alphaBoolBlend {blendOption=Linear}
    boneToModify {boneName=FACIAL_L_Jawline}
    rotation {roll=20}
   }
   n64 = "Transform (Modify) Bone - Bone: FACIAL_R_Jawline"!(ComponentPose=@n63, bAlphaBoolEnabled:bool=True, Translation="0,0,0", Rotation="0,0,20") -> (Pose) {
    alphaInputType=Bool
    rotationMode=BMM_Replace
    rotationSpace=BCS_ParentBoneSpace
    translationMode=BMM_Replace
    translationSpace=BCS_ParentBoneSpace
    alphaBoolBlend {blendOption=Linear}
    boneToModify {boneName=FACIAL_R_Jawline}
    rotation {roll=20}
   }
   n65 = "Transform (Modify) Bone - Bone: FACIAL_C_Jaw"!(ComponentPose=@n64, bAlphaBoolEnabled:bool=True, Translation="0,0,0", Rotation="0,0,20") -> (Pose) {
    alphaInputType=Bool
    rotationMode=BMM_Replace
    rotationSpace=BCS_ParentBoneSpace
    translationMode=BMM_Replace
    translationSpace=BCS_ParentBoneSpace
    alphaBoolBlend {blendOption=Linear}
    boneToModify {boneName=FACIAL_C_Jaw}
    rotation {roll=20}
   }
   n66 = "Transform (Modify) Bone - Bone: FACIAL_C_Jawline"!(ComponentPose=@n65, bAlphaBoolEnabled:bool=True, Translation="0,0,0", Rotation="0,0,20") -> (Pose) {
    alphaInputType=Bool
    rotationMode=BMM_Replace
    rotationSpace=BCS_ParentBoneSpace
    translationMode=BMM_Replace
    translationSpace=BCS_ParentBoneSpace
    alphaBoolBlend {blendOption=Linear}
    boneToModify {boneName=FACIAL_C_Jawline}
    rotation {roll=20}
   }
   n67 = "Constraint - mstr_pistol (PP)"!(ComponentPose=@n66, Alpha=1, ConstraintWeights_0=0.2, ConstraintWeights_1:float=0.8) -> (Pose) {
    alphaBoolBlend {blendOption=Linear}
    boneToModify {boneName=mstr_pistol}
    constraintSetup [{targetBone {boneName=pelvis}}, {targetBone {boneName=thigh_twist_01_l}}]
    constraintWeights [1,1]
   }
   Save cached pose 'CONTROL RIG POSE'(Pose=Component To Local(ComponentPose=@n67)) {cacheName=CONTROL RIG POSE}
   n68 = "Constraint - dyn_belt_01 (PP)"!(
    ComponentPose=Local To Component(LocalPose=Use cached pose 'SPINTING INPUT'())
    Alpha=0.3
    ConstraintWeights_0=0.6
    ConstraintWeights_1:float=0.4
   ) -> (Pose) {
    alphaBoolBlend {blendOption=Linear}
    boneToModify {boneName=dyn_belt_01}
    constraintSetup [{targetBone {boneName=thigh_twist_01_r}}, {targetBone {boneName=pelvis}}]
    constraintWeights [1,1]
   }
   n69 = "Constraint - dyn_belt_02 (PP)"!(ComponentPose=@n68, Alpha=0.4, ConstraintWeights_0=0.7, ConstraintWeights_1:float=0.3) -> (Pose) {
    alphaBoolBlend {blendOption=Linear}
    boneToModify {boneName=dyn_belt_02}
    constraintSetup [{targetBone {boneName=thigh_twist_01_r}}, {targetBone {boneName=pelvis}}]
    constraintWeights [1,1]
   }
   n70 = "Constraint - dyn_belt_03 (PP)"!(ComponentPose=@n69, Alpha=0.5, ConstraintWeights_0=0.8, ConstraintWeights_1:float=0.2) -> (Pose) {
    alphaBoolBlend {blendOption=Linear}
    boneToModify {boneName=dyn_belt_03}
    constraintSetup [{targetBone {boneName=thigh_twist_01_r}}, {targetBone {boneName=pelvis}}]
    constraintWeights [1,1]
   }
   n71 = "Constraint - dyn_belt_04 (PP)"!(ComponentPose=@n70, Alpha=0.6, ConstraintWeights_0=0.9, ConstraintWeights_1:float=0.1) -> (Pose) {
    alphaBoolBlend {blendOption=Linear}
    boneToModify {boneName=dyn_belt_04}
    constraintSetup [{targetBone {boneName=thigh_twist_01_r}}, {targetBone {boneName=pelvis}}]
    constraintWeights [1,1]
   }
   n72 = "Constraint - dyn_belt_05 (PP)"!(ComponentPose=@n71, Alpha=0.7, ConstraintWeights_0=1, ConstraintWeights_1:float=0) -> (Pose) {
    alphaBoolBlend {blendOption=Linear}
    boneToModify {boneName=dyn_belt_05}
    constraintSetup [{targetBone {boneName=thigh_twist_01_r}}, {targetBone {boneName=pelvis}}]
    constraintWeights [1,1]
   }
   Save cached pose 'Blend FE pose'(Pose=Blend Poses by bool(
     BlendPose_0=Component To Local(ComponentPose=@n72)
     BlendPose_1=Use cached pose 'SPINTING INPUT'()
     BlendTime_0=0.3
     BlendTime_1=0.3
     bActiveValue=Get bIsSprinting()
    ) {blendType=Linear}) {cacheName=Blend FE pose}
   n73 = "Constraint - deform_lower_belt_front (PPPPP)"!(
    ComponentPose=Local To Component(LocalPose=Use cached pose 'CONTROL RIG POSE'())
    Alpha=1
    ConstraintWeights_0=0.1
    ConstraintWeights_1:float=0.1
    ConstraintWeights_2:float=0.1
    ConstraintWeights_3:float=0.1
    ConstraintWeights_4:float=0.6
   ) -> (Pose) {
    alphaBoolBlend {blendOption=Linear}
    boneToModify {boneName=deform_lower_belt_front}
    constraintSetup [
     {targetBone {boneName=thigh_l}}
     {targetBone {boneName=thigh_r}}
     {targetBone {boneName=deform_groin_l}}
     {targetBone {boneName=deform_groin_r}}
     {targetBone {boneName=pelvis}}
    ]
    constraintWeights [1,1,1,1,1]
   }
   n74 = "Constraint - deform_lower_belt_front_left (PP)"!(ComponentPose=@n73, Alpha=1, ConstraintWeights_0=0.5, ConstraintWeights_1:float=0.5) -> (Pose) {
    alphaBoolBlend {blendOption=Linear}
    boneToModify {boneName=deform_lower_belt_front_left}
    constraintSetup [{targetBone {boneName=thigh_l}}, {targetBone {boneName=thigh_twist_01_l}}]
    constraintWeights [1,1]
   }
   n75 = "Constraint - deform_lower_belt_left (PP)"!(ComponentPose=@n74, Alpha=1, ConstraintWeights_0=0.5, ConstraintWeights_1:float=0.5) -> (Pose) {
    alphaBoolBlend {blendOption=Linear}
    boneToModify {boneName=deform_lower_belt_left}
    constraintSetup [{targetBone {boneName=thigh_l}}, {targetBone {boneName=thigh_twist_01_l}}]
    constraintWeights [1,1]
   }
   n76 = "Transform (Modify) Bone - Bone: deform_lower_belt_front"(ComponentPose=@n75, Alpha=1, Translation="-0.000,-0.000,5.652", Rotation="0,0,0", Scale="1,1,1") {
    rotationMode=BMM_Additive
    rotationSpace=BCS_BoneSpace
    translationMode=BMM_Additive
    translationSpace=BCS_BoneSpace
    alphaBoolBlend {blendOption=Linear}
    boneToModify {boneName=deform_lower_belt_front}
    translation {z=5.652}
   }
   n77 = "Transform (Modify) Bone - Bone: deform_lower_belt_front_left"(ComponentPose=@n76, Alpha=1, Translation="2.887,-0.319,0.921", Rotation="-0.002,-7.917,-0.000", Scale="1,1,1") {
    rotationMode=BMM_Additive
    rotationSpace=BCS_BoneSpace
    translationMode=BMM_Additive
    translationSpace=BCS_BoneSpace
    alphaBoolBlend {blendOption=Linear}
    boneToModify {boneName=deform_lower_belt_front_left}
    rotation=<rot -0.002 -7.917 0>
    translation=<2.887 -0.319 0.921>
   }
   n78 = "Transform (Modify) Bone - Bone: deform_lower_belt_left"(ComponentPose=@n77, Alpha=1, Translation="-0.362,-1.076,1.963", Rotation="0,0,0", Scale="1,1,1") {
    rotationMode=BMM_Additive
    rotationSpace=BCS_BoneSpace
    translationMode=BMM_Additive
    translationSpace=BCS_BoneSpace
    alphaBoolBlend {blendOption=Linear}
    boneToModify {boneName=deform_lower_belt_left}
    translation=<-0.362 -1.076 1.963>
   }
   n79 = Get Alpha thigh_l() -> (Alpha thigh_l:float)
   n80 = "Transform (Modify) Bone - Bone: deform_lower_belt_front_left"(ComponentPose=@n78, Alpha=@n79, Translation="5.924,-1.950,2.877", Rotation="-0.002,-7.917,-0.000", Scale="1,1,1") {
    rotationMode=BMM_Additive
    rotationSpace=BCS_BoneSpace
    translationMode=BMM_Additive
    translationSpace=BCS_BoneSpace
    alphaBoolBlend {blendOption=Linear}
    boneToModify {boneName=deform_lower_belt_front_left}
    rotation=<rot -0.002 -7.917 0>
    translation=<5.924 -1.95 2.877>
   }
   n81 = "Transform (Modify) Bone - Bone: deform_lower_belt_front"(ComponentPose=@n80, Alpha=@n79, Translation="9.599,-3.113,0.016", Rotation="0,0,0", Scale="1,1,1") {
    rotationMode=BMM_Additive
    rotationSpace=BCS_BoneSpace
    translationMode=BMM_Additive
    translationSpace=BCS_BoneSpace
    alphaBoolBlend {blendOption=Linear}
    boneToModify {boneName=deform_lower_belt_front}
    translation=<9.599 -3.113 0.016>
   }
   n82 = "Constraint - deform_lower_belt_front (TTTTTTT)"!(
    ComponentPose=Local To Component(LocalPose=Use cached pose 'CONTROL RIG POSE'())
    Alpha=1
    ConstraintWeights_0=0.127
    ConstraintWeights_1:float=0.085
    ConstraintWeights_2:float=0.609
    ConstraintWeights_3:float=0
    ConstraintWeights_4:float=0
    ConstraintWeights_5:float=0.176
    ConstraintWeights_6:float=0.003
   ) -> (Pose) {
    alphaBoolBlend {blendOption=Linear}
    boneToModify {boneName=deform_lower_belt_front}
    constraintSetup [
     {transformType=Translation, targetBone {boneName=deform_groin_l}}
     {transformType=Translation, targetBone {boneName=deform_groin_r}}
     {transformType=Translation, targetBone {boneName=pelvis}}
     {transformType=Translation, targetBone {boneName=spine_02}}
     {transformType=Translation, targetBone {boneName=thigh_l}}
     {transformType=Translation, targetBone {boneName=thigh_twist_01_l}}
     {transformType=Translation, targetBone {boneName=thigh_twist_01_r}}
    ]
    constraintWeights [1,1,1,1,1,1,1]
   }
   n83 = "Constraint - deform_lower_belt_front_left (PP)"(ComponentPose=@n82, Alpha=1, ConstraintWeights_0=0.5, ConstraintWeights_1=0.5) {
    alphaBoolBlend {blendOption=Linear}
    boneToModify {boneName=deform_lower_belt_front_left}
    constraintSetup [{targetBone {boneName=thigh_l}}, {targetBone {boneName=thigh_twist_01_l}}]
    constraintWeights [1,1]
   }
   n84 = "Constraint - deform_lower_belt_left (PP)"(ComponentPose=@n83, Alpha=1, ConstraintWeights_0=0.5, ConstraintWeights_1=0.5) {
    alphaBoolBlend {blendOption=Linear}
    boneToModify {boneName=deform_lower_belt_left}
    constraintSetup [{targetBone {boneName=thigh_l}}, {targetBone {boneName=thigh_twist_01_l}}]
    constraintWeights [1,1]
   }
   n85 = Get Left Leg() -> (Left Leg:float)
   n86 = "Transform (Modify) Bone - Bone: deform_lower_belt_front_left"(ComponentPose=@n84, Alpha=@n85, Translation="-5.696,-5.420,1.168", Rotation="0,0,0", Scale="1,1,1") {
    rotationMode=BMM_Additive
    rotationSpace=BCS_BoneSpace
    translationMode=BMM_Additive
    translationSpace=BCS_BoneSpace
    alphaBoolBlend {blendOption=Linear}
    boneToModify {boneName=deform_lower_belt_front_left}
    translation=<-5.696 -5.42 1.168>
   }
   n87 = "Transform (Modify) Bone - Bone: deform_lower_belt_front_left"(ComponentPose=@n86, Alpha=1, Translation="11.313,4.747,-3.483", Rotation="-0.817,-27.042,36.631", Scale="1,1,1") {
    rotationMode=BMM_Additive
    rotationSpace=BCS_BoneSpace
    translationMode=BMM_Additive
    translationSpace=BCS_BoneSpace
    alphaBoolBlend {blendOption=Linear}
    boneToModify {boneName=deform_lower_belt_front_left}
    rotation=<rot -0.817 -27.042 36.631>
    translation=<11.313 4.747 -3.483>
   }
   n88 = "Transform (Modify) Bone - Bone: deform_lower_belt_left"(ComponentPose=@n87, Alpha=1, Translation="-1.488,6.320,2.530", Rotation="-0.084,-0.119,0.274", Scale="1,1,1") {
    rotationMode=BMM_Additive
    rotationSpace=BCS_BoneSpace
    translationMode=BMM_Additive
    translationSpace=BCS_BoneSpace
    alphaBoolBlend {blendOption=Linear}
    boneToModify {boneName=deform_lower_belt_left}
    rotation=<rot -0.084 -0.119 0.274>
    translation=<-1.488 6.32 2.53>
   }
   n89 = "Transform (Modify) Bone - Bone: deform_lower_belt_front"(ComponentPose=@n88, Alpha=@n85, Translation="0.967,-0.797,0.984", Rotation="0.012,-15.546,-0.001", Scale="1,1,1") {
    rotationMode=BMM_Additive
    rotationSpace=BCS_BoneSpace
    translationMode=BMM_Additive
    translationSpace=BCS_BoneSpace
    alphaBoolBlend {blendOption=Linear}
    boneToModify {boneName=deform_lower_belt_front}
    rotation=<rot 0.012 -15.546 -0.001>
    translation=<0.967 -0.797 0.984>
   }
   n90 = "Transform (Modify) Bone - Bone: dyn_belt_mstr"(ComponentPose=@n89, Alpha=1, Translation="-0.846,0.464,0.501", Rotation="46.537,31.155,-39.875", Scale="1,1,1") {
    rotationMode=BMM_Additive
    rotationSpace=BCS_BoneSpace
    translationMode=BMM_Additive
    translationSpace=BCS_BoneSpace
    alphaBoolBlend {blendOption=Linear}
    boneToModify {boneName=dyn_belt_mstr}
    rotation=<rot 46.537 31.155 -39.875>
    translation=<-0.846 0.464 0.501>
   }
   n91 = Blend Poses by bool(
    BlendPose_0=Component To Local(ComponentPose=@n90)
    BlendPose_1=Component To Local(ComponentPose=@n81)
    BlendTime_0=0.3
    BlendTime_1=0.3
    bActiveValue=Get bIsCrouching()
   ) {blendType=Linear}
   n92 = "Constraint - deform_lower_belt_front (PPPPP)"(
    ComponentPose=Local To Component(LocalPose=Use cached pose 'CONTROL RIG POSE'())
    Alpha=1
    ConstraintWeights_0=0.1
    ConstraintWeights_1=0.1
    ConstraintWeights_2=0.1
    ConstraintWeights_3=0.1
    ConstraintWeights_4=0.6
   ) {
    alphaBoolBlend {blendOption=Linear}
    boneToModify {boneName=deform_lower_belt_front}
    constraintSetup [
     {targetBone {boneName=thigh_l}}
     {targetBone {boneName=thigh_r}}
     {targetBone {boneName=deform_groin_l}}
     {targetBone {boneName=deform_groin_r}}
     {targetBone {boneName=pelvis}}
    ]
    constraintWeights [1,1,1,1,1]
   }
   n93 = "Constraint - deform_lower_belt_front_left (PP)"(ComponentPose=@n92, Alpha=1, ConstraintWeights_0=0.5, ConstraintWeights_1=0.5) {
    alphaBoolBlend {blendOption=Linear}
    boneToModify {boneName=deform_lower_belt_front_left}
    constraintSetup [{targetBone {boneName=thigh_l}}, {targetBone {boneName=thigh_twist_01_l}}]
    constraintWeights [1,1]
   }
   n94 = "Constraint - deform_lower_belt_left (PP)"(ComponentPose=@n93, Alpha=1, ConstraintWeights_0=0.5, ConstraintWeights_1=0.5) {
    alphaBoolBlend {blendOption=Linear}
    boneToModify {boneName=deform_lower_belt_left}
    constraintSetup [{targetBone {boneName=thigh_l}}, {targetBone {boneName=thigh_twist_01_l}}]
    constraintWeights [1,1]
   }
   n95 = "Transform (Modify) Bone - Bone: deform_lower_belt_front"(ComponentPose=@n94, Alpha=1, Translation="-0.000,-0.000,5.652", Rotation="0,0,0", Scale="1,1,1") {
    rotationMode=BMM_Additive
    rotationSpace=BCS_BoneSpace
    translationMode=BMM_Additive
    translationSpace=BCS_BoneSpace
    alphaBoolBlend {blendOption=Linear}
    boneToModify {boneName=deform_lower_belt_front}
    translation {z=5.652}
   }
   n96 = "Transform (Modify) Bone - Bone: deform_lower_belt_front_left"(ComponentPose=@n95, Alpha=1, Translation="2.887,-0.319,0.921", Rotation="-0.002,-7.917,-0.000", Scale="1,1,1") {
    rotationMode=BMM_Additive
    rotationSpace=BCS_BoneSpace
    translationMode=BMM_Additive
    translationSpace=BCS_BoneSpace
    alphaBoolBlend {blendOption=Linear}
    boneToModify {boneName=deform_lower_belt_front_left}
    rotation=<rot -0.002 -7.917 0>
    translation=<2.887 -0.319 0.921>
   }
   n97 = "Transform (Modify) Bone - Bone: deform_lower_belt_left"(ComponentPose=@n96, Alpha=1, Translation="-0.362,-1.076,1.963", Rotation="0,0,0", Scale="1,1,1") {
    rotationMode=BMM_Additive
    rotationSpace=BCS_BoneSpace
    translationMode=BMM_Additive
    translationSpace=BCS_BoneSpace
    alphaBoolBlend {blendOption=Linear}
    boneToModify {boneName=deform_lower_belt_left}
    translation=<-0.362 -1.076 1.963>
   }
   n98 = "Transform (Modify) Bone - Bone: mstr_pistol"(ComponentPose=@n97, Alpha=1, Translation="-0.580,-1.015,-0.000", Rotation="0,0,0", Scale="1,1,1") {
    rotationMode=BMM_Additive
    rotationSpace=BCS_BoneSpace
    translationMode=BMM_Additive
    translationSpace=BCS_BoneSpace
    alphaBoolBlend {blendOption=Linear}
    boneToModify {boneName=mstr_pistol}
    translation=<-0.58 -1.015>
   }
   n99 = Component To Local(ComponentPose=@n98)
   n100 = Get bIsSkyDiving()
   n101 = RigidBody!(
    ComponentPose=Local To Component(LocalPose=Blend Poses by bool(BlendPose_0=@n99, BlendPose_1=@n91, BlendTime_0=0.3, BlendTime_1=0.3, bActiveValue=@n100) {blendType=Linear})
    Alpha=Get Alpha Skydiving() -> (Alpha Skydiving:float)
    OverrideWorldGravity:Vector=Get Gravity Override()
    ExternalForce
    SimSpaceSettings:Sim Space Settings Structure=(WorldAlpha=0,VelocityScaleZ=1,DampingAlpha=1,MaxLinearVelocity=10000,MaxAngularVelocity=10000,MaxLinearAcceleration=10000,MaxAngularAcceleration=10000,ExternalLinearDragV=(X=0,Y=0,Z=0),ExternalLinearVelocity=(X=0,Y=0,Z=0),ExternalAngularVelocity=(X=0,Y=0,Z=0))
   ) -> (Pose) {
    bOverrideWorldGravity=true
    overridePhysicsAsset=&/F_MED_LacedCrimp_Physics
    simulationSpace=BaseBoneSpace
    alphaBoolBlend {blendOption=Linear}
    baseBoneRef {boneName=pelvis}
    componentAppliedLinearAccClamp=<5000 5000 4999>
    overrideWorldGravity {z=-980}
   }
   Save cached pose 'SPINTING INPUT'(Pose=Component To Local(ComponentPose=@n101)) {cacheName=SPINTING INPUT}
  ]}
 EventGraph {nodes [
   n1 = Event OnCharacterCustomizationLoaded()
   > Set Body Mesh(Body Mesh:Skeletal Mesh Component Object Reference=GetOwnerSkeletalMeshComponent()) -> (Output_Get:Skeletal Mesh Component Object Reference)
   > Set HeadMesh(HeadMesh:Skeletal Mesh Component Object Reference=Get Array Item with EFort Custom Part Type(Array=@n1.CustomMeshParts, PartType=Head)) -> (Output_Get:Skeletal Mesh Component Object Reference)
   Event BlueprintUpdateAnimation()
   > n2 = Sequence()
   n3 = Get HeadMesh() -> (HeadMesh:Skeletal Mesh Component Object Reference)
   > Is Valid(InputObject=@n3)
   n4 = GetSocketBoneName!(self:Skinned Mesh Component Object Reference=@n3, InSocketName:Name=FACIAL_C_FacialRoot) -> (ReturnValue:Name=None)
   > Set is face bones(is face bones:bool="Not Equal Exactly (String)"(A="To String (Name)"(InName=@n4), B=None)) -> (Output_Get:bool)
   n5 = Get HeadMesh()
   n6 = BreakTransform(InTransform=GetSocketTransform(self=@n5, InSocketName=FACIAL_C_UnderChin, TransformSpace=RTS_ParentBoneSpace))
   > Set FACIAL_C_UnderChin_Translation(FACIAL_C_UnderChin_Translation:Vector=@n6) -> (Output_Get:Vector)
   > Set FACIAL_C_UnderChin_Rotation(FACIAL_C_UnderChin_Rotation:Rotator=@n6.Rotation) -> (Output_Get:Rotator)
   n7 = BreakTransform(InTransform=GetSocketTransform(self=@n5, InSocketName=FACIAL_L_UnderChin, TransformSpace=RTS_ParentBoneSpace))
   > Set FACIAL_L_UnderChin_Translation(FACIAL_L_UnderChin_Translation:Vector=@n7) -> (Output_Get:Vector)
   > Set FACIAL_L_UnderChin_Rotation(FACIAL_L_UnderChin_Rotation:Rotator=@n7.Rotation) -> (Output_Get:Rotator)
   n8 = BreakTransform(InTransform=GetSocketTransform(self=@n5, InSocketName=FACIAL_R_UnderChin, TransformSpace=RTS_ParentBoneSpace))
   > Set FACIAL_R_UnderChin_Translation(FACIAL_R_UnderChin_Translation:Vector=@n8) -> (Output_Get:Vector)
   > Set FACIAL_R_UnderChin_Rotation(FACIAL_R_UnderChin_Rotation:Rotator=@n8.Rotation) -> (Output_Get:Rotator)
   n9 = Get HeadMesh()
   n10 = BreakTransform(InTransform=GetSocketTransform(self=@n9, InSocketName=FACIAL_C_Jaw, TransformSpace=RTS_ParentBoneSpace))
   > Set FACIAL_C_Jaw_Tranform(FACIAL_C_Jaw_Tranform:Vector=@n10) -> (Output_Get:Vector)
   > Set FACIAL_C_Jaw_Rotation(FACIAL_C_Jaw_Rotation:Rotator=@n10.Rotation) -> (Output_Get:Rotator)
   n11 = BreakTransform(InTransform=GetSocketTransform(self=@n9, InSocketName=FACIAL_C_Jawline, TransformSpace=RTS_ParentBoneSpace))
   > Set FACIAL_C_Jawline_Translation(FACIAL_C_Jawline_Translation:Vector=@n11) -> (Output_Get:Vector)
   > Set FACIAL_C_Jawline_Rotation(FACIAL_C_Jawline_Rotation:Rotator=@n11.Rotation) -> (Output_Get:Rotator)
   n12 = BreakTransform(InTransform=GetSocketTransform(self=@n9, InSocketName=FACIAL_L_Jawline, TransformSpace=RTS_ParentBoneSpace))
   > Set FACIAL_L_Jawline_Translation(FACIAL_L_Jawline_Translation:Vector=@n12) -> (Output_Get:Vector)
   > Set FACIAL_L_Jawline_Rotation(FACIAL_L_Jawline_Rotation:Rotator=@n12.Rotation) -> (Output_Get:Rotator)
   n13 = BreakTransform(InTransform=GetSocketTransform(self=@n9, InSocketName=FACIAL_R_Jawline, TransformSpace=RTS_ParentBoneSpace))
   > Set FACIAL_R_Jawline_Translation(FACIAL_R_Jawline_Translation:Vector=@n13) -> (Output_Get:Vector)
   > Set FACIAL_R_Jawline_Rotation(FACIAL_R_Jawline_Rotation:Rotator=@n13.Rotation) -> (Output_Get:Rotator)
   n14 = Gravity Override State Function DEPRECATED(
    execute=@n2.then_1
    CharacterPartAnimInstance=Self-Reference()
    GameTimeInSeconds=GetGameTimeInSeconds()
    GravityOverrideParams=Get Gravity Override Params()
    RBAN Params=Get RBAN Params()
   )
   > Set Gravity Override(Gravity Override=@n14)
   > Set RBAN Settings Out(RBAN Settings Out:RBANSetting Struct Structure=@n14.RBAN SettingsOut) -> (Output_Get:RBANSetting Struct Structure)
   n15 = Get Body Mesh() -> (Body Mesh:Skeletal Mesh Component Object Reference)
   > Set thigh_l_alpha_z(thigh_l_alpha_z:float=GetAlphaFromReferenceTransformDelta(SkeletalMeshComponent=@n15, BoneName=thigh_l, BaseName=pelvis, DeltaComponentType=EulerX, InputRange=(X=0,Y=-30), TwistAxis="0,0,1")) -> (Output_Get:float)
   > Set thigh_l_alpha_x(thigh_l_alpha_x:float=GetAlphaFromReferenceTransformDelta(SkeletalMeshComponent=@n15, BoneName=thigh_l, BaseName=pelvis, InputRange=(X=-100,Y=0), TwistAxis="0,0,1")) -> (Output_Get:float)
   > Set Left Leg(Left Leg:float=GetAlphaFromReferenceTransformDelta(SkeletalMeshComponent=@n15, BoneName=thigh_l, BaseName=pelvis, InputRange=(X=-80,Y=-45), TwistAxis="0,0,1")) -> (Output_Get:float)
   n16 = Get Body Mesh()
   > Set Alpha upperarm_l(Alpha upperarm_l:float=GetAlphaFromReferenceTransformDelta(SkeletalMeshComponent=@n16, BoneName=upperarm_l, BaseName=clavicle_l, DeltaComponentType=EulerY, InputRange=(X=0,Y=45), TwistAxis="0,0,1")) -> (Output_Get:float)
   > Set Alpha upperarm_r(Alpha upperarm_r:float=GetAlphaFromReferenceTransformDelta(SkeletalMeshComponent=@n16, BoneName=upperarm_r, BaseName=clavicle_r, DeltaComponentType=EulerY, InputRange=(X=0,Y=45), TwistAxis="0,0,1")) -> (Output_Get:float)
   > Set Alpha thigh_l(Alpha thigh_l:float=GetAlphaFromReferenceTransformDelta(SkeletalMeshComponent=@n16, BoneName=thigh_l, BaseName=pelvis, InputRange=(X=-40,Y=-70), TwistAxis="0,0,1")) -> (Output_Get:float)
   > Set jump easing(jump easing:double=MapRangeClamped(Value=Get PawnVelocityZ(), InRangeA=630, InRangeB=-630, OutRangeA=15, OutRangeB=-15)) -> (Output_Get:double)
   > Set side jump easing(side jump easing:double=MapRangeClamped(Value=Get PawnVelocityZ(), InRangeA=630, InRangeB=-630, OutRangeA=8, OutRangeB=-8)) -> (Output_Get:double)
   > Set jump easing tr(jump easing tr:double=MapRangeClamped(Value=Get PawnVelocityZ(), InRangeA=630, InRangeB=-630, OutRangeA=2, OutRangeB=0)) -> (Output_Get:double)
   > Set is in jump or falling(is in jump or falling:bool=OR Boolean(A=Get bIsFalling(), B=Get bIsJumping())) -> (Output_Get:bool)
   > Set Alpha Skydiving(Alpha Skydiving:float=SelectFloat(A=0.4, B=1, bPickA=Get bIsSkyDiving())) -> (Output_Get:float)
   > Set Alpha Sprinting(Alpha Sprinting:double=SelectFloat(A=1, B=0, bPickA=Get bIsSprinting())) -> (Output_Get:double)
   > Set 1-2-3 LODs alpha(1-2-3 LODs alpha:bool="integer > integer"(A=Get CurrentLOD())) -> (Output_Get:bool)
  ]}
]
variables [
 1-2-3 LODs alpha {type=boolean}
 Alpha Skydiving {type=number}
 Alpha Sprinting {type=number}
 Alpha thigh_l {type=number}
 Alpha upperarm_l {type=number}
 Alpha upperarm_r {type=number}
 Body Mesh {type=SkeletalMeshComponent}
 FACIAL_C_Jaw_Rotation {type=Rotator}
 FACIAL_C_Jaw_Tranform {type=Vector}
 FACIAL_C_Jawline_Rotation {type=Rotator}
 FACIAL_C_Jawline_Translation {type=Vector}
 FACIAL_C_UnderChin_Rotation {type=Rotator}
 FACIAL_C_UnderChin_Translation {type=Vector}
 FACIAL_L_Jawline_Rotation {type=Rotator}
 FACIAL_L_Jawline_Translation {type=Vector}
 FACIAL_L_UnderChin_Rotation {type=Rotator}
 FACIAL_L_UnderChin_Translation {type=Vector}
 FACIAL_R_Jawline_Rotation {type=Rotator}
 FACIAL_R_Jawline_Translation {type=Vector}
 FACIAL_R_UnderChin_Rotation {type=Rotator}
 FACIAL_R_UnderChin_Translation {type=Vector}
 Gravity Override {type=Vector}
 Gravity Override Params {type=GravityOverrideParamsStruct}
 HeadMesh {type=SkeletalMeshComponent}
 is face bones {type=boolean}
 is in jump or falling {type=boolean}
 jump easing {type=number}
 jump easing tr {type=number}
 Left Leg {type=number}
 RBAN Params {
  type=RBANParamsStruct
  default {
   dBNO {simSpaceSettings {
     maxAngularAcceleration=1000
     maxAngularVelocity=3
     maxLinearAcceleration=300
     maxLinearVelocity=700
     velocityScaleZ=0
     externalLinearDragV=<0 0.5 0>
    }}
   dBNO_Carried {simSpaceSettings {
     maxAngularAcceleration=1000
     maxAngularVelocity=3
     maxLinearAcceleration=300
     maxLinearVelocity=700
     velocityScaleZ=0
     externalLinearDragV=<0 0.5 0>
    }}
   emoteOrMelee {simSpaceSettings {
     maxAngularAcceleration=1000
     maxAngularVelocity=3
     maxLinearAcceleration=300
     maxLinearVelocity=700
     velocityScaleZ=0
     externalLinearDragV=<0 0.5 0>
    }}
   front_End {simSpaceSettings {
     maxAngularAcceleration=50
     maxAngularVelocity=1.5
     maxLinearAcceleration=150
     maxLinearVelocity=350
     velocityScaleZ=0
     externalLinearDragV=<0 0.5 0>
    }}
   onGround_Crouch_Moving {simSpaceSettings {
     maxAngularAcceleration=1000
     maxAngularVelocity=3
     maxLinearAcceleration=300
     maxLinearVelocity=700
     velocityScaleZ=0
     externalLinearDragV=<0 0.5 0>
    }}
   onGround_Falling {simSpaceSettings {
     maxAngularAcceleration=1000
     maxAngularVelocity=3
     maxLinearAcceleration=300
     maxLinearVelocity=700
     velocityScaleZ=0
     externalLinearDragV=<0 0.5 0>
    }}
   onGround_Moving {simSpaceSettings {
     maxAngularAcceleration=1000
     maxAngularVelocity=3
     maxLinearAcceleration=300
     maxLinearVelocity=700
     velocityScaleZ=0
     externalLinearDragV=<0 0.5 0>
    }}
   onGround_Standing {simSpaceSettings {
     maxAngularAcceleration=1000
     maxAngularVelocity=3
     maxLinearAcceleration=300
     maxLinearVelocity=700
     velocityScaleZ=0
     externalLinearDragV=<0 0.5 0>
    }}
   rail_Grinding {simSpaceSettings {
     maxAngularAcceleration=1000
     maxAngularVelocity=3
     maxLinearAcceleration=300
     maxLinearVelocity=700
     velocityScaleZ=0
     worldAlpha=0.1
     externalLinearDragV=<0 0.5 0>
    }}
   riding_Moving {simSpaceSettings {
     maxAngularAcceleration=1000
     maxAngularVelocity=3
     maxLinearAcceleration=300
     maxLinearVelocity=700
     velocityScaleZ=0
     externalLinearDragV=<0 0.5 0>
    }}
   riding_Moving_Targeting {simSpaceSettings {
     maxAngularAcceleration=1000
     maxAngularVelocity=3
     maxLinearAcceleration=300
     maxLinearVelocity=700
     velocityScaleZ=0
     externalLinearDragV=<0 0.5 0>
    }}
   riding_Standing {simSpaceSettings {
     maxAngularAcceleration=1000
     maxAngularVelocity=3
     maxLinearAcceleration=300
     maxLinearVelocity=700
     velocityScaleZ=0
     externalLinearDragV=<0 0.5 0>
    }}
   skydive_DivingDown {simSpaceSettings {
     maxAngularAcceleration=1000
     maxAngularVelocity=3
     maxLinearAcceleration=300
     maxLinearVelocity=700
     velocityScaleZ=0
     externalLinearDragV=<0 0.5 0>
    }}
   skydive_FloatingDown {
    componentLinearAccScale=<0 0.5 0>
    componentLinearVelScale=<0 0.5 0>
    simSpaceSettings {
     maxAngularAcceleration=1000
     maxAngularVelocity=3
     maxLinearAcceleration=300
     maxLinearVelocity=700
     velocityScaleZ=0
     externalLinearDragV=<0 0.5 0>
    }
   }
   skydive_GoingUp {simSpaceSettings {
     maxAngularAcceleration=1000
     maxAngularVelocity=3
     maxLinearAcceleration=300
     maxLinearVelocity=700
     velocityScaleZ=0
     externalLinearDragV=<0 0.5 0>
    }}
   skydive_Parachuting {
    componentLinearAccScale=<0 0.1 0.3>
    componentLinearVelScale=<0 0.1 0.3>
    simSpaceSettings {
     maxAngularAcceleration=1000
     maxAngularVelocity=3
     maxLinearAcceleration=300
     maxLinearVelocity=700
     velocityScaleZ=0
     externalLinearDragV=<0 0.5 0>
    }
   }
   skydive_Surfing {simSpaceSettings {
     maxAngularAcceleration=1000
     maxAngularVelocity=3
     maxLinearAcceleration=300
     maxLinearVelocity=700
     velocityScaleZ=0
     externalLinearDragV=<0 0.5 0>
    }}
   swimming {simSpaceSettings {
     maxAngularAcceleration=1000
     maxAngularVelocity=3
     maxLinearAcceleration=300
     maxLinearVelocity=700
     velocityScaleZ=0
     externalLinearDragV=<0 0.5 0>
    }}
   tactical_Sprint {simSpaceSettings {
     maxAngularAcceleration=1000
     maxAngularVelocity=3
     maxLinearAcceleration=300
     maxLinearVelocity=700
     velocityScaleZ=0
     externalLinearDragV=<0 0.5 0>
    }}
   ziplining {simSpaceSettings {
     maxAngularAcceleration=1000
     maxAngularVelocity=3
     maxLinearAcceleration=300
     maxLinearVelocity=700
     velocityScaleZ=0
     externalLinearDragV=<0 0.5 0>
    }}
  }
 }
 RBAN Settings Out {
  type=RBANSettingStruct
  default {
   componentLinearAccScale=<0 0 0>
   componentLinearVelScale=<0 0 0>
   simSpaceSettings {
    dampingAlpha=1
    maxAngularAcceleration=1000
    maxAngularVelocity=3
    maxLinearAcceleration=300
    maxLinearVelocity=700
    velocityScaleZ=0
    worldAlpha=1
    externalAngularVelocity=<0 0 0>
    externalLinearDragV=<0 0.5 0>
    externalLinearVelocity=<0 0 0>
   }
  }
 }
 side jump easing {type=number}
 thigh_l_alpha_x {type=number}
 thigh_l_alpha_z {type=number}
]
