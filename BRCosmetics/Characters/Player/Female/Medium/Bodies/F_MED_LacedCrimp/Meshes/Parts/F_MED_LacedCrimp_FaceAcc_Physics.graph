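# Physics asset for the LacedCrimp female medium character's FaceAcc mesh part, setting up physics for its dynamic hair bones.
# Seven short chains (two mid, five back) each link a base sphere to two capsules, with locked twist and small per-joint swing limit angles.
# Two capsules on the head bone form the head volume, and no constraint joins the chain bases to it.
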PhysicsAsset /BRCosmetics/Characters/Player/Female/Medium/Bodies/F_MED_LacedCrimp/Meshes/Parts/F_MED_LacedCrimp_FaceAcc_Physics
bodies {
 dyn_hairBack_01_Base_l [{shapeType=Sphere, radius=0.5, shapeName=dyn_hairMid_01_Base_l_sphere}]
 dyn_hairBack_01_Base_l1 [{shapeType=Sphere, radius=0.5, shapeName=dyn_hairMid_01_Base_l_sphere}]
 dyn_hairBack_01_Base_l2 [{shapeType=Sphere, radius=0.5, shapeName=dyn_hairMid_01_Base_l_sphere}]
 dyn_hairBack_01_Base_r [{shapeType=Sphere, radius=0.5, shapeName=dyn_hairMid_01_Base_l_sphere}]
 dyn_hairBack_01_c [{shapeType=Capsule, length=2.9169, radius=0.9392, shapeName=dyn_hairBack_01_r_capsule, center=<2.4318 0 0>, rotation=<rot 90 0 0>}]
 dyn_hairBack_01_c_BASE [{shapeType=Sphere, radius=0.5, shapeName=dyn_hairBack_01_c_BASE_sphere}]
 dyn_hairBack_01_l [{shapeType=Capsule, length=2.9169, radius=0.9392, shapeName=dyn_hairBack_01_r_capsule, center=<2.4318 0 0>, rotation=<rot 90 0 0>}]
 dyn_hairBack_01_l1 [{shapeType=Capsule, length=2.9169, radius=0.9392, shapeName=dyn_hairBack_01_r_capsule, center=<2.4318 0 0>, rotation=<rot 90 0 0>}]
 dyn_hairBack_01_l2 [{shapeType=Capsule, length=2.9169, radius=0.9392, shapeName=dyn_hairBack_01_r_capsule, center=<2.4318 0 0>, rotation=<rot 90 0 0>}]
 dyn_hairBack_01_r [{shapeType=Capsule, length=2.9169, radius=0.9392, shapeName=dyn_hairBack_01_r_capsule, center=<2.4318 0 0>, rotation=<rot 90 0 0>}]
 dyn_hairBack_02_c [{shapeType=Capsule, length=2.9169, radius=0.9392, shapeName=dyn_hairBack_01_r_capsule, center=<2.4318 0 0>, rotation=<rot 90 0 0>}]
 dyn_hairBack_02_l [{shapeType=Capsule, length=2.9169, radius=0.9392, shapeName=dyn_hairBack_01_r_capsule, center=<2.4318 0 0>, rotation=<rot 90 0 0>}]
 dyn_hairBack_02_l1 [{shapeType=Capsule, length=2.9169, radius=0.9392, shapeName=dyn_hairBack_01_r_capsule, center=<2.4318 0 0>, rotation=<rot 90 0 0>}]
 dyn_hairBack_02_l2 [{shapeType=Capsule, length=2.9169, radius=0.9392, shapeName=dyn_hairBack_01_r_capsule, center=<2.4318 0 0>, rotation=<rot 90 0 0>}]
 dyn_hairBack_02_r [{shapeType=Capsule, length=2.9169, radius=0.9392, shapeName=dyn_hairBack_01_r_capsule, center=<2.4318 0 0>, rotation=<rot 90 0 0>}]
 dyn_hairMID_01_Base_r [{shapeType=Sphere, radius=0.5, shapeName=dyn_hairMid_01_Base_l_sphere}]
 dyn_hairMid_01_Base_l [{shapeType=Sphere, radius=0.5, shapeName=dyn_hairMid_01_Base_l_sphere}]
 dyn_hairMid_01_l [{shapeType=Capsule, length=2.9169, radius=0.9392, shapeName=dyn_hairBack_01_r_capsule, center=<2.4318 0 0>, rotation=<rot 90 0 0>}]
 dyn_hairMid_01_r [{shapeType=Capsule, length=2.9169, radius=0.9392, shapeName=dyn_hairBack_01_r_capsule, center=<2.4318 0 0>, rotation=<rot 90 0 0>}]
 dyn_hairMid_02_l [{shapeType=Capsule, length=2.9169, radius=0.9392, shapeName=dyn_hairBack_01_r_capsule, center=<2.4318 0 0>, rotation=<rot 90 0 0>}]
 dyn_hairMid_02_r [{shapeType=Capsule, length=2.9169, radius=0.9392, shapeName=dyn_hairBack_01_r_capsule, center=<2.4318 0 0>, rotation=<rot 90 0 0>}]
 head [
  {shapeType=Capsule, length=3.8965, radius=5.8522, center=<3.3425 3.5392 0>, rotation=<rot 90 -179.1048 -173.3028>}
  {shapeType=Capsule, length=6.8514, radius=4.8393, shapeName=head_capsule, center=<0.756 4.3931 0>, rotation=<rot 90 -179.1048 -173.9807>}
 ]
}
constraints [
 {
  bone1Name=dyn_hairMid_02_l
  bone2Name=dyn_hairMid_01_l
  swing1LimitDegrees=30
  swing2LimitDegrees=30
  twistLimitDegrees=45
  twistMotion=Locked
 }
 {
  bone1Name=dyn_hairMid_01_l
  bone2Name=dyn_hairMid_01_Base_l
  swing1LimitDegrees=17
  swing2LimitDegrees=5
  twistLimitDegrees=45
  twistMotion=Locked
 }
 {
  bone1Name=dyn_hairMid_02_r
  bone2Name=dyn_hairMid_01_r
  swing1LimitDegrees=30
  swing2LimitDegrees=30
  twistLimitDegrees=45
  twistMotion=Locked
 }
 {
  bone1Name=dyn_hairMid_01_r
  bone2Name=dyn_hairMID_01_Base_r
  swing1LimitDegrees=17
  swing2LimitDegrees=5
  twistLimitDegrees=45
  twistMotion=Locked
 }
 {
  bone1Name=dyn_hairBack_02_c
  bone2Name=dyn_hairBack_01_c
  swing1LimitDegrees=30
  swing2LimitDegrees=30
  twistLimitDegrees=45
  twistMotion=Locked
 }
 {
  bone1Name=dyn_hairBack_01_c
  bone2Name=dyn_hairBack_01_c_BASE
  swing1LimitDegrees=20
  swing2LimitDegrees=5
  twistLimitDegrees=45
  twistMotion=Locked
 }
 {
  bone1Name=dyn_hairBack_02_r
  bone2Name=dyn_hairBack_01_r
  swing1LimitDegrees=30
  swing2LimitDegrees=5
  twistLimitDegrees=45
  twistMotion=Locked
 }
 {
  bone1Name=dyn_hairBack_01_r
  bone2Name=dyn_hairBack_01_Base_r
  swing1LimitDegrees=5
  swing2LimitDegrees=5
  twistLimitDegrees=45
  twistMotion=Locked
 }
 {
  bone1Name=dyn_hairBack_02_l1
  bone2Name=dyn_hairBack_01_l1
  swing1LimitDegrees=5
  swing2LimitDegrees=30
  twistLimitDegrees=45
  twistMotion=Locked
 }
 {
  bone1Name=dyn_hairBack_01_l1
  bone2Name=dyn_hairBack_01_Base_l1
  swing1LimitDegrees=5
  swing2LimitDegrees=5
  twistLimitDegrees=45
  twistMotion=Locked
 }
 {
  bone1Name=dyn_hairBack_02_l
  bone2Name=dyn_hairBack_01_l
  swing1LimitDegrees=5
  swing2LimitDegrees=30
  twistLimitDegrees=45
  twistMotion=Locked
 }
 {
  bone1Name=dyn_hairBack_01_l
  bone2Name=dyn_hairBack_01_Base_l
  swing1LimitDegrees=5
  swing2LimitDegrees=5
  twistLimitDegrees=45
  twistMotion=Locked
 }
 {
  bone1Name=dyn_hairBack_02_l2
  bone2Name=dyn_hairBack_01_l2
  swing1LimitDegrees=5
  swing2LimitDegrees=5
  twistLimitDegrees=45
  twistMotion=Locked
 }
 {
  bone1Name=dyn_hairBack_01_l2
  bone2Name=dyn_hairBack_01_Base_l2
  swing1LimitDegrees=5
  swing2LimitDegrees=5
  twistLimitDegrees=45
  twistMotion=Locked
 }
]
