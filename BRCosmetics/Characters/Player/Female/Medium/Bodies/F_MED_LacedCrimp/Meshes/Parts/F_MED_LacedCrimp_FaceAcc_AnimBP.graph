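# Post-process-style AnimBP for the LacedCrimp face accessory, adding secondary motion to hair and keeping hood bones matched to the body mesh.
# It copies the parent pose, pins three hood bones to BodyMesh world transforms, then runs RigidBody simulation.
# The cached result feeds two AnimDynamics hair chains, selected by bIsInFrontEnd.
# Head look-up scales down three hair bones; updates also compute state-based gravity and RBAN settings.
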AnimBlueprint /BRCosmetics/Characters/Player/Female/Medium/Bodies/F_MED_LacedCrimp/Meshes/Parts/F_MED_LacedCrimp_FaceAcc_AnimBP
target_skeleton=&/F_MED_LacedCrimp_FaceAcc_Skeleton
graphs [
 AnimGraph {nodes [
   n1 = Anim Dynamics - Bone: dyn_hairTop_01_c!(ComponentPose=Local To Component(LocalPose=Use cached pose 'Blend FE pose'()), Alpha=1, GravityOverride:Vector="0,0,-980") -> (Pose) {
    angularDampingOverride=0.9999
    angularSpringConstant=2000
    bAngularSpring=true
    bGravityOverrideInSimSpace=true
    bOverrideAngularDamping=true
    bUseGravityOverride=true
    simulationSpace=World
    alphaBoolBlend {blendOption=Linear}
    boundBone {boneName=dyn_hairTop_01_c}
    componentAppliedLinearAccClamp=<1000 1000 1000>
    componentLinearAccScale=<0.1 0.1 0.1>
    componentLinearVelScale=<0.2 0.2 0.2>
    gravityOverride {z=-980}
    physicsBodyDefinitions [{
      boundBone {boneName=dyn_hairTop_01_c}
      boxExtents=<10 10 10>
      constraintSetup {angularLimitsMax=<0 20 10>, angularLimitsMin=<0 -20 -10>, angularTarget=<0 0 0>}
      localJointOffset=<1 0 0>
     }]
   }
   n2 = Anim Dynamics - Bone: dyn_hairBack_01_l2!(ComponentPose=@n1, Alpha=1, GravityOverride:Vector="0,0,-980") -> (Pose) {
    angularDampingOverride=0.9999
    angularSpringConstant=1500
    bAngularSpring=true
    bGravityOverrideInSimSpace=true
    bOverrideAngularDamping=true
    bUseGravityOverride=true
    simulationSpace=World
    alphaBoolBlend {blendOption=Linear}
    boundBone {boneName=dyn_hairBack_01_l2}
    componentAppliedLinearAccClamp=<1000 1000 1000>
    componentLinearAccScale=<0.1 0.1 0.1>
    componentLinearVelScale=<0.2 0.2 0.2>
    gravityOverride {z=-980}
    physicsBodyDefinitions [{
      boundBone {boneName=dyn_hairBack_01_l2}
      boxExtents=<20 10 10>
      constraintSetup {angularLimitsMax=<0 0 0>, angularLimitsMin=<0 -5 -5>, angularTarget=<0 0 0>}
      localJointOffset=<1 0 -0.2>
     }]
   }
   n3 = Anim Dynamics - Bone: dyn_hairBack_02_l2!(ComponentPose=@n2, Alpha=1, GravityOverride:Vector="0,0,-980") -> (Pose) {
    angularDampingOverride=0.9999
    angularSpringConstant=1000
    bAngularSpring=true
    bGravityOverrideInSimSpace=true
    bOverrideAngularDamping=true
    bUseGravityOverride=true
    simulationSpace=World
    alphaBoolBlend {blendOption=Linear}
    boundBone {boneName=dyn_hairBack_02_l2}
    componentAppliedLinearAccClamp=<1000 1000 1000>
    componentLinearAccScale=<0.1 0.1 0.1>
    componentLinearVelScale=<0.2 0.2 0.2>
    gravityOverride {z=-980}
    physicsBodyDefinitions [{
      boundBone {boneName=dyn_hairBack_02_l2}
      boxExtents=<10 5 5>
      constraintSetup {angularLimitsMax=<0 0 0>, angularLimitsMin=<0 -5 -5>, angularTarget=<0 0 0>}
      localJointOffset=<1 0 0>
     }]
   }
   n4 = Component To Local(ComponentPose=@n3)
   n5 = Anim Dynamics - Bone: dyn_hairFront_01_c1!(ComponentPose=Local To Component(LocalPose=Use cached pose 'Blend FE pose'()), Alpha=1, GravityOverride:Vector="0,0,-980") -> (Pose) {
    angularDampingOverride=0.9999
    angularSpringConstant=2000
    bAngularSpring=true
    bGravityOverrideInSimSpace=true
    bOverrideAngularDamping=true
    bUseGravityOverride=true
    alphaBoolBlend {blendOption=Linear}
    boundBone {boneName=dyn_hairFront_01_c1}
    componentAppliedLinearAccClamp=<3000 3000 3000>
    componentLinearAccScale=<0.5 0.5 0.5>
    componentLinearVelScale=<0.5 0.5 0.5>
    gravityOverride {z=-980}
    physicsBodyDefinitions [{
      boundBone {boneName=dyn_hairFront_01_c1}
      boxExtents=<10 10 10>
      constraintSetup {angularLimitsMax=<0 0 2>, angularLimitsMin=<0 -15 -2>, angularTarget=<0 0 0>}
      localJointOffset=<1.5 0 0>
     }]
   }
   n6 = Anim Dynamics - Bone: dyn_hairTop_01_c(ComponentPose=@n5, Alpha=1, GravityOverride="0,0,-980") {
    angularDampingOverride=0.9999
    angularSpringConstant=2000
    bAngularSpring=true
    bGravityOverrideInSimSpace=true
    bOverrideAngularDamping=true
    bUseGravityOverride=true
    alphaBoolBlend {blendOption=Linear}
    boundBone {boneName=dyn_hairTop_01_c}
    componentAppliedLinearAccClamp=<3000 3000 3000>
    componentLinearAccScale=<0.5 0.5 0.5>
    componentLinearVelScale=<0.5 0.5 0.5>
    gravityOverride {z=-980}
    physicsBodyDefinitions [{
      boundBone {boneName=dyn_hairTop_01_c}
      boxExtents=<10 10 10>
      constraintSetup {angularLimitsMax=<0 20 10>, angularLimitsMin=<0 -20 -10>, angularTarget=<0 0 0>}
      localJointOffset=<1.5 0 0>
     }]
   }
   n7 = Anim Dynamics - Bone: dyn_hairBack_01_l2(ComponentPose=@n6, Alpha=1, GravityOverride="0,0,-980") {
    angularDampingOverride=0.9999
    angularSpringConstant=1500
    bAngularSpring=true
    bGravityOverrideInSimSpace=true
    bOverrideAngularDamping=true
    bUseGravityOverride=true
    alphaBoolBlend {blendOption=Linear}
    boundBone {boneName=dyn_hairBack_01_l2}
    componentAppliedLinearAccClamp=<3000 3000 3000>
    componentLinearAccScale=<0.5 0.5 0.5>
    componentLinearVelScale=<0.5 0.5 0.5>
    gravityOverride {z=-980}
    physicsBodyDefinitions [{
      boundBone {boneName=dyn_hairBack_01_l2}
      boxExtents=<20 10 10>
      constraintSetup {angularLimitsMax=<0 0 0>, angularLimitsMin=<0 -5 -5>, angularTarget=<0 0 0>}
      localJointOffset=<2 0 -0.2>
     }]
   }
   n8 = Anim Dynamics - Bone: dyn_hairBack_02_l2(ComponentPose=@n7, Alpha=1, GravityOverride="0,0,-980") {
    angularDampingOverride=0.9999
    angularSpringConstant=1000
    bAngularSpring=true
    bGravityOverrideInSimSpace=true
    bOverrideAngularDamping=true
    bUseGravityOverride=true
    alphaBoolBlend {blendOption=Linear}
    boundBone {boneName=dyn_hairBack_02_l2}
    componentAppliedLinearAccClamp=<3000 3000 3000>
    componentLinearAccScale=<0.5 0.5 0.5>
    componentLinearVelScale=<0.5 0.5 0.5>
    gravityOverride {z=-980}
    physicsBodyDefinitions [{
      boundBone {boneName=dyn_hairBack_02_l2}
      boxExtents=<10 5 5>
      constraintSetup {angularLimitsMax=<0 0 0>, angularLimitsMin=<0 -5 -5>, angularTarget=<0 0 0>}
      localJointOffset=<1 0 0>
     }]
   }
   n9 = Component To Local(ComponentPose=@n8)
   n10 = Get bIsInFrontEnd()
   n11 = Get HeadLookUpAlpha() -> (HeadLookUpAlpha:float)
   n12 = "Transform (Modify) Bone - Bone: dyn_hairBack_02_c"(
    ComponentPose=Local To Component(LocalPose=Blend Poses by bool(BlendPose_0=@n4, BlendPose_1=@n9, bActiveValue=@n10) {blendType=Linear})
    Alpha=@n11
    Translation="0,0,0"
    Rotation="0,0,0"
    Scale="0.8,0.8,0.8"
   ) {
    scaleMode=BMM_Additive
    scaleSpace=BCS_BoneSpace
    alphaBoolBlend {blendOption=Linear}
    boneToModify {boneName=dyn_hairBack_02_c}
    scale=<0.8 0.8 0.8>
   }
   n13 = "Transform (Modify) Bone - Bone: dyn_hairMid_02_r"(ComponentPose=@n12, Alpha=@n11, Translation="0,0,0", Rotation="0,0,0", Scale="0.8,0.8,0.8") {
    scaleMode=BMM_Additive
    scaleSpace=BCS_BoneSpace
    alphaBoolBlend {blendOption=Linear}
    boneToModify {boneName=dyn_hairMid_02_r}
    scale=<0.8 0.8 0.8>
   }
   n14 = "Transform (Modify) Bone - Bone: dyn_hairBack_02_r"(ComponentPose=@n13, Alpha=@n11, Translation="0,0,0", Rotation="0,0,0", Scale="0.8,0.8,0.8") {
    scaleMode=BMM_Additive
    scaleSpace=BCS_BoneSpace
    alphaBoolBlend {blendOption=Linear}
    boneToModify {boneName=dyn_hairBack_02_r}
    scale=<0.8 0.8 0.8>
   }
   Output Pose(Result=Component To Local(ComponentPose=@n14)) {layerGroup=None}
   n15 = "Transform (Modify) Bone - Bone: dyn_hood_l_01"(
    ComponentPose=Local To Component(LocalPose=Copy Pose From Mesh() {bUseAttachedParent=true})
    Alpha=1
    Translation=Get lf Hood translation() -> (lf Hood translation:Vector)
    Rotation=Get lf Hood rotation() -> (lf Hood rotation:Rotator)
    Scale="1,1,1"
   ) {
    rotationMode=BMM_Replace
    rotationSpace=BCS_WorldSpace
    scaleMode=BMM_Additive
    translationMode=BMM_Replace
    translationSpace=BCS_WorldSpace
    alphaBoolBlend {blendInTime=1, blendOption=Linear, blendOutTime=1}
    boneToModify {boneName=dyn_hood_l_01}
    rotation {roll=20}
   }
   n16 = "Transform (Modify) Bone - Bone: dyn_hood_r_01"(
    ComponentPose=@n15
    Alpha=1
    Translation=Get rt Hood translation() -> (rt Hood translation:Vector)
    Rotation=Get rt Hood rotation() -> (rt Hood rotation:Rotator)
    Scale="1,1,1"
   ) {
    rotationMode=BMM_Replace
    rotationSpace=BCS_WorldSpace
    scaleMode=BMM_Additive
    translationMode=BMM_Replace
    translationSpace=BCS_WorldSpace
    alphaBoolBlend {blendInTime=1, blendOption=Linear, blendOutTime=1}
    boneToModify {boneName=dyn_hood_r_01}
    rotation {roll=20}
   }
   n17 = "Transform (Modify) Bone - Bone: dyn_hoodieTop"(
    ComponentPose=@n16
    Alpha=1
    Translation=Get top hood translation() -> (top hood translation:Vector)
    Rotation=Get Top Hood rotation() -> (Top Hood rotation:Rotator)
    Scale="1,1,1"
   ) {
    rotationMode=BMM_Replace
    rotationSpace=BCS_WorldSpace
    scaleMode=BMM_Additive
    translationMode=BMM_Replace
    translationSpace=BCS_WorldSpace
    alphaBoolBlend {blendInTime=1, blendOption=Linear, blendOutTime=1}
    boneToModify {boneName=dyn_hoodieTop}
    rotation {roll=20}
   }
   n18 = RigidBody!(
    ComponentPose=@n17
    Alpha=0.8
    OverrideWorldGravity:Vector="0,0,0"
    ExternalForce
    SimSpaceSettings:Sim Space Settings Structure=(WorldAlpha=1,VelocityScaleZ=1,DampingAlpha=1,MaxLinearVelocity=500,MaxAngularVelocity=3,MaxLinearAcceleration=50,MaxAngularAcceleration=1,ExternalLinearDragV=(X=0,Y=0,Z=0),ExternalLinearVelocity=(X=0,Y=0,Z=0),ExternalAngularVelocity=(X=0,Y=0,Z=0))
   ) -> (Pose) {
    bClampLinearTranslationLimitToRefPose=true
    bOverrideWorldGravity=true
    overridePhysicsAsset=&/F_MED_LacedCrimp_FaceAcc_Physics
    simulationSpace=BaseBoneSpace
    alphaBoolBlend {blendOption=Linear}
    baseBoneRef {boneName=spine_03}
    componentAppliedLinearAccClamp=<3000 3000 3000>
    simSpaceSettings {maxAngularAcceleration=1, maxAngularVelocity=3, maxLinearAcceleration=50, maxLinearVelocity=500, worldAlpha=1}
   }
   Save cached pose 'Blend FE pose'(Pose=Component To Local(ComponentPose=@n18)) {cacheName=Blend FE pose}
  ]}
 EventGraph {nodes [
   Event BlueprintUpdateAnimation()
   > n1 = Gravity Override State Function DEPRECATED(
    CharacterPartAnimInstance=Self-Reference()
    GameTimeInSeconds=GetGameTimeInSeconds()
    GravityOverrideParams=Get Gravity Override Params()
    RBAN Params=Get RBAN Params()
   )
   > Set Gravity Override(Gravity Override=@n1)
   > Set RBAN Settings Out(RBAN Settings Out:RBANSetting Struct Structure=@n1.RBAN SettingsOut) -> (Output_Get:RBANSetting Struct Structure)
   n2 = BreakTransform(InTransform=GetSocketTransform(self=Get BodyMesh() -> (BodyMesh:Skeletal Mesh Component Object Reference), InSocketName=dyn_hood_l_01, TransformSpace=RTS_World))
   > Set lf Hood translation(lf Hood translation:Vector=@n2) -> (Output_Get:Vector)
   > Set lf Hood rotation(lf Hood rotation:Rotator=@n2.Rotation) -> (Output_Get:Rotator)
   n3 = BreakTransform(InTransform=GetSocketTransform(self=Get BodyMesh(), InSocketName=dyn_hood_r_01, TransformSpace=RTS_World))
   > Set rt Hood translation(rt Hood translation:Vector=@n3) -> (Output_Get:Vector)
   > Set rt Hood rotation(rt Hood rotation:Rotator=@n3.Rotation) -> (Output_Get:Rotator)
   n4 = BreakTransform(InTransform=GetSocketTransform(self=Get BodyMesh(), InSocketName=dyn_hoodieTop, TransformSpace=RTS_World))
   > Set top hood translation(top hood translation:Vector=@n4) -> (Output_Get:Vector)
   > Set Top Hood rotation(Top Hood rotation:Rotator=@n4.Rotation) -> (Output_Get:Rotator)
   > Set HeadLookUpAlpha(HeadLookUpAlpha:float=GetAlphaFromReferenceTransformDelta(SkeletalMeshComponent=Get BodyMesh(), BoneName=head, BaseName=spine_05, InputRange=(X=0,Y=-40), TwistAxis="1,0,0")) -> (Output_Get:float)
   > Set crouching Alpha(crouching Alpha:float=SelectFloat(A=0, B=1, bPickA=Get bIsCrouching())) -> (Output_Get:float)
   n5 = Event OnCharacterCustomizationLoaded()
   > Set BodyMesh(BodyMesh=Get Array Item with EFort Custom Part Type(Array=@n5.CustomMeshParts))
  ]}
]
variables [
 BodyMesh {type=SkeletalMeshComponent}
 crouching Alpha {type=number}
 Gravity Override {type=Vector}
 Gravity Override Params {type=GravityOverrideParamsStruct}
 HeadLookUpAlpha {type=number}
 lf Hood rotation {type=Rotator}
 lf Hood translation {type=Vector}
 RBAN Params {
  type=RBANParamsStruct
  default {
   dBNO {simSpaceSettings {
     maxAngularAcceleration=1000
     maxAngularVelocity=3
     maxLinearAcceleration=300
     maxLinearVelocity=700
     velocityScaleZ=0
     externalLinearDragV=<0 0.5 0>
    }}
   dBNO_Carried {simSpaceSettings {
     maxAngularAcceleration=1000
     maxAngularVelocity=3
     maxLinearAcceleration=300
     maxLinearVelocity=700
     velocityScaleZ=0
     externalLinearDragV=<0 0.5 0>
    }}
   emoteOrMelee {simSpaceSettings {
     maxAngularAcceleration=1000
     maxAngularVelocity=3
     maxLinearAcceleration=300
     maxLinearVelocity=700
     velocityScaleZ=0
     externalLinearDragV=<0 0.5 0>
    }}
   front_End {simSpaceSettings {
     maxAngularAcceleration=200
     maxAngularVelocity=2
     maxLinearAcceleration=200
     maxLinearVelocity=500
     velocityScaleZ=0
     externalLinearDragV=<0 0.5 0>
    }}
   onGround_Crouch_Moving {simSpaceSettings {
     maxAngularAcceleration=1000
     maxAngularVelocity=3
     maxLinearAcceleration=300
     maxLinearVelocity=700
     velocityScaleZ=0
     externalLinearDragV=<0 0.5 0>
    }}
   onGround_Falling {simSpaceSettings {
     maxAngularAcceleration=1000
     maxAngularVelocity=3
     maxLinearAcceleration=300
     maxLinearVelocity=700
     velocityScaleZ=0
     externalLinearDragV=<0 0.5 0>
    }}
   onGround_Moving {simSpaceSettings {
     maxAngularAcceleration=1000
     maxAngularVelocity=3
     maxLinearAcceleration=300
     maxLinearVelocity=700
     velocityScaleZ=0
     externalLinearDragV=<0 0.5 0>
    }}
   onGround_Standing {simSpaceSettings {
     maxAngularAcceleration=1000
     maxAngularVelocity=3
     maxLinearAcceleration=300
     maxLinearVelocity=700
     velocityScaleZ=0
     externalLinearDragV=<0 0.5 0>
    }}
   rail_Grinding {simSpaceSettings {
     maxAngularAcceleration=1000
     maxAngularVelocity=3
     maxLinearAcceleration=300
     maxLinearVelocity=700
     velocityScaleZ=0
     worldAlpha=0.1
     externalLinearDragV=<0 0.5 0>
    }}
   riding_Moving {simSpaceSettings {
     maxAngularAcceleration=1000
     maxAngularVelocity=3
     maxLinearAcceleration=300
     maxLinearVelocity=700
     velocityScaleZ=0
     externalLinearDragV=<0 0.5 0>
    }}
   riding_Moving_Targeting {simSpaceSettings {
     maxAngularAcceleration=1000
     maxAngularVelocity=3
     maxLinearAcceleration=300
     maxLinearVelocity=700
     velocityScaleZ=0
     externalLinearDragV=<0 0.5 0>
    }}
   riding_Standing {simSpaceSettings {
     maxAngularAcceleration=1000
     maxAngularVelocity=3
     maxLinearAcceleration=300
     maxLinearVelocity=700
     velocityScaleZ=0
     externalLinearDragV=<0 0.5 0>
    }}
   skydive_DivingDown {simSpaceSettings {
     maxAngularAcceleration=1000
     maxAngularVelocity=3
     maxLinearAcceleration=300
     maxLinearVelocity=700
     velocityScaleZ=0
     externalLinearDragV=<0 0.5 0>
    }}
   skydive_FloatingDown {
    componentLinearAccScale=<0 0.5 0>
    componentLinearVelScale=<0 0.5 0>
    simSpaceSettings {
     maxAngularAcceleration=1000
     maxAngularVelocity=3
     maxLinearAcceleration=300
     maxLinearVelocity=700
     velocityScaleZ=0
     externalLinearDragV=<0 0.5 0>
    }
   }
   skydive_GoingUp {simSpaceSettings {
     maxAngularAcceleration=1000
     maxAngularVelocity=3
     maxLinearAcceleration=300
     maxLinearVelocity=700
     velocityScaleZ=0
     externalLinearDragV=<0 0.5 0>
    }}
   skydive_Parachuting {
    componentLinearAccScale=<0 0.1 0.3>
    componentLinearVelScale=<0 0.1 0.3>
    simSpaceSettings {
     maxAngularAcceleration=1000
     maxAngularVelocity=3
     maxLinearAcceleration=300
     maxLinearVelocity=700
     velocityScaleZ=0
     externalLinearDragV=<0 0.5 0>
    }
   }
   skydive_Surfing {simSpaceSettings {
     maxAngularAcceleration=1000
     maxAngularVelocity=3
     maxLinearAcceleration=300
     maxLinearVelocity=700
     velocityScaleZ=0
     externalLinearDragV=<0 0.5 0>
    }}
   swimming {simSpaceSettings {
     maxAngularAcceleration=1000
     maxAngularVelocity=3
     maxLinearAcceleration=300
     maxLinearVelocity=700
     velocityScaleZ=0
     externalLinearDragV=<0 0.5 0>
    }}
   tactical_Sprint {simSpaceSettings {
     maxAngularAcceleration=1000
     maxAngularVelocity=3
     maxLinearAcceleration=300
     maxLinearVelocity=700
     velocityScaleZ=0
     externalLinearDragV=<0 0.5 0>
    }}
   ziplining {simSpaceSettings {
     maxAngularAcceleration=1000
     maxAngularVelocity=3
     maxLinearAcceleration=300
     maxLinearVelocity=700
     velocityScaleZ=0
     externalLinearDragV=<0 0.5 0>
    }}
  }
 }
 RBAN Settings Out {
  type=RBANSettingStruct
  default {
   componentLinearAccScale=<0 0 0>
   componentLinearVelScale=<0 0 0>
   simSpaceSettings {
    dampingAlpha=1
    maxAngularAcceleration=3
    maxAngularVelocity=3
    maxLinearAcceleration=250
    maxLinearVelocity=500
    velocityScaleZ=1
    worldAlpha=1
    externalAngularVelocity=<0 0 0>
    externalLinearDragV=<0 0.5 0>
    externalLinearVelocity=<0 0 0>
   }
  }
 }
 rt Hood rotation {type=Rotator}
 rt Hood translation {type=Vector}
 Top Hood rotation {type=Rotator}
 top hood translation {type=Vector}
]
